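# HLODWater is a two-sided Single Layer Water surface material, named for HLOD water and enabled for Nanite, instanced and spline meshes.
# WaterTwoTexturePanner supplies normals and WPO; normals get an XY gain and normalization before joining color, specular, roughness and opacity in WaterQualityOptimizations.
# Scattering color is scaled by its alpha, while absorption divides by squared color.
# Phase G comes from MPC_Water, leaving the local Anisotropy parameter unconnected.
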Material /FortWater/Materials/WaterSurface/HLODWater
graph {
 properties {
  allowTranslucentCustomDepthWrites=true
  bComputeFogPerPixel=true
  bUsedWithInstancedStaticMeshes=true
  bUsedWithNanite=true
  bUsedWithSplineMeshes=true
  bUsedWithWater=true
  pixelDepthOffsetMode=PDOM_Legacy
  previewMesh=/Engine/Transient.MaterialEditorMeshComponent_2
  shadingModel=MSM_SingleLayerWater
  twoSided=true
  usedShadingModels=SingleLayerWater
 }
 nodes [
  n1 = MaterialFunctionCall!(Normal, WPO, Far Mask Override, Additional Mask, LowDetailBiome_Mask) {materialFunction=/FortWater/Materials/Functions/WaterTwoTexturePanner}
  n2 = ScalarParameter() {defaultValue=5, parameterName=NormalGain}
  n3 = MaterialFunctionCall(X=@n2, Y=@n2, Z=Constant() {r=1}) {materialFunction=&MakeFloat3}
  n4 = SetMaterialAttributes!(
   MaterialAttributes
   Base Color=VectorParameter() {parameterName=BaseColor}.RGB
   Specular=ScalarParameter() {defaultValue=0.225, group=Water Shading Model, parameterName=Specular}
   Roughness=ScalarParameter() {defaultValue=0.05, group=Water Shading Model, parameterName=Roughness}
   Opacity=ScalarParameter() {group=Water Shading Model, parameterName=Surface Opacity}
   Normal=Normalize(VectorInput=Multiply(A=@n1.Normal, B=@n3.Result))
   World Position Offset=@n1.WPO
  ) {attributeSetTypes [
    #1
    #2
    #3
    #4
    #5
    #6
   ]}
  n5 = GetMaterialAttributes(None=MaterialFunctionCall!(Attributes=@n4) {desc=Simple Spec and Roughness On Mobile, materialFunction=/FortWater/Materials/Functions/WaterQualityOptimizations}.Result) {attributeGetTypes [
    #1
    #2
    #3
    #4
    #5
    #6
   ]}
  MaterialFunctionCall!(Normal=@n5.Normal) {materialFunction=/FortWater/Materials/Functions/MF_WaveFlatten}
  n6 = ComponentMask(None=CameraVectorWS()) {b=true}
  n7 = VectorParameter() {group=Water Shading Model, parameterName=Scattering, defaultValue=<rgba 0.1268 0.1336 0.15 0.5>}
  n8 = Multiply(A=@n7.RGB, B=Divide(A=@n7.A) {constB=100})
  Multiply(A=Saturate(None=Ceil(None=@n6)), B=@n8)
  Multiply(A=@n5.Normal, B=TwoSidedSign())
  ScalarParameter() {defaultValue=0.3, group=Water Shading Model, parameterName=Anisotropy}
  n9 = VectorParameter() {group=Water Shading Model, parameterName=Absorption, defaultValue=<rgba 1 13.35 20 100>}
  SingleLayerWaterMaterialOutput!(
   ScatteringCoefficients=@n8
   AbsorptionCoefficients=Divide(B=Multiply(A=@n9.RGB, B=@n9.RGB))
   PhaseG=CollectionParameter() {collection=/Water/Materials/MPC/MPC_Water, parameterName=Ansotropy}
   ColorScaleBehindWater
  )
 ]
}
