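# Material function that hides objects sitting below the global water level by adjusting World Position Offset in a passed-in attribute set.
# It compares the terrain collection's WaterLevel, plus a Water Vanish Threshold, against the actor's world Z, using Ceil and Saturate to form a binary mask.
# When flooded, the mask pushes the whole mesh far down in Z, toggled by the HideUnderGlobalWaterLevel static switch, on by default.
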MaterialFunction /FortWater/Materials/Functions/WaterFloodMask
graph {nodes [
  n1 = GetMaterialAttributes(None=FunctionInput() {inputName=Attributes, inputType=FunctionInput_MaterialAttributes}) {attributeGetTypes [#1]}
  n2 = CollectionParameter() {collection=/Game/Athena/Apollo/Environments/Landscape/MPC/Apollo_MPC_Terrain, parameterName=WaterLevel}
  n3 = ComponentMask(None=ActorPositionWS()) {b=true}
  n4 = Ceil(None=Add(A=Subtract(A=@n2, B=@n3), B=ScalarParameter() {defaultValue=32, parameterName=Water Vanish Threshold}))
  n5 = Add(A=@n1.WorldPositionOffset, B=Multiply(A=Saturate(None=@n4), B=Constant3Vector() {constant=<rgba 0 0 -5000000 0>}))
  FunctionOutput(None=SetMaterialAttributes!(
    MaterialAttributes=@n1.MaterialAttributes
    WorldPositionOffset=StaticSwitchParameter(True=@n5, False=@n1.WorldPositionOffset) {defaultValue=true, parameterName=HideUnderGlobalWaterLevel}
   ) {attributeSetTypes [#1]})
 ]}
