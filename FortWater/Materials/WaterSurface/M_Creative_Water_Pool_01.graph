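# Single Layer Water surface material for Creative pools, supporting instanced and spline meshes.
# Optional two-texture panning feeds the water simulation function, whose normal joins surface attributes through a flood-mask function.
# Scattering and absorption come from color parameters, with mobile-specific boosts and a two-sided-sign scale.
# Refraction strength scales with pool depth over pixel depth and switches between top and underside values, with normals flipped on back faces.
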Material /FortWater/Materials/WaterSurface/M_Creative_Water_Pool_01
graph {
 properties {
  allowTranslucentCustomDepthWrites=true
  bUsedWithInstancedStaticMeshes=true
  bUsedWithSplineMeshes=true
  bUsedWithWater=true
  pixelDepthOffsetMode=PDOM_Legacy
  refractionMethod=RM_PixelNormalOffset
  shadingModel=MSM_SingleLayerWater
  usedShadingModels=SingleLayerWater
 }
 nodes [
  n1 = Divide(A=ScalarParameter() {defaultValue=128, parameterName=Pool Depth}, B=Divide(A=PixelDepth().R) {constB=4})
  n2 = ScalarParameter() {defaultValue=0.1, parameterName=BottomRefraction}
  n3 = ScalarParameter() {defaultValue=0.15, parameterName=Refraction}
  n4 = TwoSidedSign()
  n5 = Saturate(None=@n4)
  Add(B=Multiply(A=Saturate(None=@n1), B=LinearInterpolate(A=@n2, B=@n3, Alpha=@n5))) {constA=1}
  Constant3Vector() {constant=<rgba 0 0 1 0>}
  n6 = VectorParameter() {parameterName=EmissiveColor, defaultValue=<rgba 0 0 0 1>}
  Multiply(A=@n6.RGB, B=@n6.RGB)
  n7 = MaterialFunctionCall!(Normal, WPO, Far Mask Override=ScalarParameter() {defaultValue=1, parameterName=FarMask}, Additional Mask, LowDetailBiome_Mask) {materialFunction=/FortWater/Materials/Functions/WaterTwoTexturePanner}
  n8 = MaterialFunctionCall!(
   Flowmap Velocity
   Previous Foam
   Normal=StaticSwitchParameter(True=@n7.Normal, False=Constant3Vector() {constant=<rgba 0 0 1 0>}) {defaultValue=true, parameterName=UsePanning?}
   WPO=StaticSwitchParameter(True=@n7.WPO, False=Constant3Vector()) {defaultValue=true, parameterName=UsePanning?}
  ) {materialFunction=/FortWater/Materials/Functions/WaterSimulationTexture}
  n9 = SetMaterialAttributes!(
   MaterialAttributes
   Base Color=VectorParameter() {parameterName=BaseColor}.RGB
   Specular=ScalarParameter() {defaultValue=0.225, group=Water Shading Model, parameterName=Specular}
   Roughness=ScalarParameter() {defaultValue=0.03, group=Water Shading Model, parameterName=Roughness}
   Opacity=ScalarParameter() {group=Water Shading Model, parameterName=Surface Opacity}
   Normal=@n8.Normal
   World Position Offset=@n8.Normal
  ) {attributeSetTypes [
    #1
    #2
    #3
    #4
    #5
    #6
   ]}
  n10 = GetMaterialAttributes(None=MaterialFunctionCall!(Attributes=@n9) {materialFunction=/FortWater/Materials/Functions/WaterFloodMask}.Result) {attributeGetTypes [
    #1
    #2
    #3
    #4
    #5
    #6
   ]}
  Multiply(A=@n10.Normal, B=TwoSidedSign())
  n11 = Saturate(None=Add(A=@n4, B=ScalarParameter() {parameterName=Min Scattering Absorption}))
  n12 = VectorParameter() {group=Water Shading Model, parameterName=Scattering, defaultValue=<rgba 0.1 0.1 0.1 0.5>}
  n13 = Multiply(A=@n12.RGB, B=Divide(A=@n12.A) {constB=100})
  n14 = VectorParameter() {parameterName=MobileScatterBoost, defaultValue=<rgba 1 3 5 1.5>}
  n15 = ShadingPathSwitch(Default=Constant() {r=1}, Mobile=Multiply(A=@n14.RGB, B=@n14.RGB)) {desc=Migrated from FeatureLevelSwitch}
  n16 = VectorParameter() {group=Water Shading Model, parameterName=Absorption, defaultValue=<rgba 1 8 20 100>}
  n17 = Divide(
   A=Multiply(A=@n16.RGB, B=@n16.RGB)
   B=ShadingPathSwitch(Default=Constant() {r=1}, Mobile=ScalarParameter() {defaultValue=2, parameterName=Mobile Absorption Boost, sliderMin=1}) {desc=Migrated from FeatureLevelSwitch}
  )
  SingleLayerWaterMaterialOutput!(
   ScatteringCoefficients=Multiply(A=@n11, B=Multiply(A=@n13, B=@n15))
   AbsorptionCoefficients=Multiply(A=@n11, B=Divide(B=@n17))
   PhaseG=ScalarParameter() {defaultValue=0.3, group=Water Shading Model, parameterName=Anisotropy}
   ColorScaleBehindWater
  )
 ]
}
