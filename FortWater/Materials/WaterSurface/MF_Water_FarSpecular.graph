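# Water surface material function that adjusts specular and roughness at very far distances to better blend the ocean-sky horizon.
# It builds a radial mask from world XY distance to the world origin, not the camera, offset by StartDistance and divided by EndDistance, which thus acts as transition length.
# A squared ease-out shapes the saturated mask.
# That mask blends incoming Specular toward a multiplied value and Roughness toward FarRoughness.
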MaterialFunction /FortWater/Materials/WaterSurface/MF_Water_FarSpecular
graph {nodes [
  n1 = GetMaterialAttributes(None=FunctionInput() {inputName=Material Attributes, inputType=FunctionInput_MaterialAttributes}) {attributeGetTypes [#1, #2]}
  n2 = ScalarParameter() {
   defaultValue=0.5
   desc=Multiplier to the specular value at very far distances to better blend the ocean-sky horizon
   group=Far Specular
   parameterName=FarSpecular IntensityMultiplier
   sortPriority=0
  }
  n3 = Multiply(A=@n1.Specular, B=@n2)
  n4 = ScalarParameter() {
   defaultValue=94834.4844
   desc=Start point of the FarSpecular radial transition gradient
   group=Far Specular
   parameterName=FarSpecular StartDistance
   sortPriority=1
  }
  n5 = Subtract(A=Distance(A=Constant(), B=WorldPosition().XY), B=@n4)
  n6 = ScalarParameter() {
   defaultValue=48225.3086
   desc=End point of the FarSpecular radial transition gradient
   group=Far Specular
   parameterName=FarSpecular EndDistance
   sortPriority=2
  }
  n7 = OneMinus(None=Saturate(None=Divide(A=@n5, B=@n6)))
  n8 = OneMinus(None=Multiply(A=@n7, B=@n7))
  n9 = ScalarParameter() {defaultValue=0.3, group=Far Specular, parameterName=FarRoughness}
  FunctionOutput(None=SetMaterialAttributes!(
    MaterialAttributes=@n1.MaterialAttributes
    Specular=LinearInterpolate(A=@n1.Specular, B=@n3, Alpha=@n8)
    Roughness=LinearInterpolate(A=@n1.Roughness, B=@n9, Alpha=@n8)
   ) {attributeSetTypes [#1, #2]})
 ]}
