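# Far-distance ocean surface material using two-sided Single Layer Water, with parameter-driven scattering, absorption and anisotropy.
# WaterTwoTexturePanner supplies the normal and position offset, with the normal flipped on back faces.
# Quality and shading-path switches simplify specular and roughness on lower tiers, forward and mobile, before MF_Water_FarSpecular finishes the attributes.
# World position offset also pushes vertices beyond a cutoff distance from world origin far outward in XY.
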Material /FortWater/Materials/WaterSurface/M_Water_Ocean_02_NewModel_Far
graph {
 properties {
  allowTranslucentCustomDepthWrites=true
  bUseMaterialAttributes=true
  bUsedWithInstancedStaticMeshes=true
  bUsedWithNanite=true
  bUsedWithSplineMeshes=true
  bUsedWithWater=true
  pixelDepthOffsetMode=PDOM_Legacy
  shadingModel=MSM_SingleLayerWater
  twoSided=true
  usedShadingModels=SingleLayerWater
 }
 nodes [
  n1 = ScalarParameter() {defaultValue=0.225, group=Water Shading Model, parameterName=Specular}
  n2 = MaterialFunctionCall!(
   Normal
   WPO=VectorParameter() {parameterName=PositionOffset, defaultValue=<rgba 0 0 0 1>}.RGB
   Far Mask Override
   Additional Mask
   LowDetailBiome_Mask
  ) {materialFunction=/FortWater/Materials/Functions/WaterTwoTexturePanner}
  n3 = SetMaterialAttributes!(
   MaterialAttributes
   Base Color=VectorParameter() {parameterName=BaseColor}.RGB
   Specular=QualitySwitch(
    Default=MaterialFunctionCall!(InSpecular=@n1, WaterDepth) {materialFunction=/FortWater/Materials/Functions/MF_WaterBodySpecular}.Result
    Low=@n1
    Medium=@n1
   )
   Roughness=ScalarParameter() {defaultValue=0.05, group=Water Shading Model, parameterName=Roughness}
   Opacity=ScalarParameter() {group=Water Shading Model, parameterName=Surface Opacity}
   Normal=@n2.Normal
   World Position Offset=@n2.Normal
  ) {attributeSetTypes [
    #1
    #2
    #3
    #4
    #5
    #6
   ]}
  n4 = GetMaterialAttributes(None=@n3) {attributeGetTypes [
    #1
    #2
    #3
    #4
    #5
    #6
   ]}
  n5 = SetMaterialAttributes!(
   MaterialAttributes
   Base Color=@n4.Base Color
   Metallic=Constant()
   Specular=@n4.Base Color
   Roughness=@n4.Base Color
   Opacity=@n4.Base Color
   Normal=Multiply(A=@n4.Normal, B=TwoSidedSign())
   World Position Offset=@n4.Base Color
  ) {attributeSetTypes [
    #1
    #7
    #2
    #3
    #4
    #5
    #6
   ]}
  n6 = GetMaterialAttributes(None=@n5) {attributeGetTypes [#3, #6]}
  n7 = ScalarParameter() {defaultValue=0.1, group=Water Shading Model, parameterName=RoughnessLowDetail}
  n8 = QualitySwitch(Default=@n6.Roughness, Low=@n7, Medium=@n7)
  n9 = WorldPosition()
  n10 = Distance(A=@n9.XYZ, B=Constant())
  n11 = ScalarParameter() {defaultValue=1000000, group=Outlines Opacity, parameterName=Distant Cutoff}
  n12 = LinearInterpolate(B=AppendVector(A=Multiply(A=@n9.XY) {constB=4}, B=Constant()), Alpha=Saturate(None=Subtract(A=@n10, B=@n11) {constB=20000}))
  MaterialFunctionCall!(Material Attributes=SetMaterialAttributes!(
    MaterialAttributes=@n6.MaterialAttributes
    Roughness=ShadingPathSwitch(Default=@n8, Deferred=@n8, Forward=@n7, Mobile=@n7)
    World Position Offset=Add(A=@n6.World Position Offset, B=@n12)
   ) {attributeSetTypes [#3, #6]}) {materialFunction=&/MF_Water_FarSpecular}
  MaterialFunctionCall!(Normal=@n4.Normal) {materialFunction=/FortWater/Materials/Functions/MF_WaveFlatten}
  n13 = VectorParameter() {group=Water Shading Model, parameterName=Scattering, defaultValue=<rgba 0.1268 0.1336 0.15 0.5>}
  n14 = VectorParameter() {group=Water Shading Model, parameterName=Absorption, defaultValue=<rgba 1 13.35 20 100>}
  SingleLayerWaterMaterialOutput!(
   ScatteringCoefficients=Multiply(A=@n13.RGB, B=Divide(A=@n13.A) {constB=100})
   AbsorptionCoefficients=Divide(B=Multiply(A=@n14.RGB, B=@n14.RGB))
   PhaseG=ScalarParameter() {defaultValue=0.3, group=Water Shading Model, parameterName=Anisotropy}
   ColorScaleBehindWater
  )
 ]
}
