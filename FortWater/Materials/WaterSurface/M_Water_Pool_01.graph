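# Two-sided Single Layer Water material for pool surfaces, with HLOD color and opacity overrides.
# Two-texture panner normals feed the water simulation texture, switchable to vertex-color flow motion, with optional distance, quality and mobile specular.
# TwoSidedSign adjusts normal, roughness, refraction and coefficients for the underside, and refraction also scales with pool depth over pixel depth.
# Scattering and absorption get mobile boosts; scattering and optional time-of-day emissive use eye-adaptation compensation.
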Material /FortWater/Materials/WaterSurface/M_Water_Pool_01
graph {
 properties {
  allowTranslucentCustomDepthWrites=true
  bUseMaterialAttributes=true
  bUsedWithInstancedStaticMeshes=true
  bUsedWithNanite=true
  bUsedWithSplineMeshes=true
  bUsedWithWater=true
  pixelDepthOffsetMode=PDOM_Legacy
  refractionMethod=RM_PixelNormalOffset
  shadingModel=MSM_SingleLayerWater
  twoSided=true
  usedShadingModels=SingleLayerWater
 }
 nodes [
  Constant3Vector() {constant=<rgba 0 0 1 0>}
  DepthFade()
  n1 = VertexColor()
  n2 = Constant() {r=1}
  LinearInterpolate(
   B=If(
    A=@n1.A
    B=Constant() {r=0.9}
    "A > B"=@n2
    "A == B"=@n2
    "A < B"=MaterialFunctionCall(Alpha Threshold=@n1.A) {materialFunction=&DitherTemporalAA_Fortnite}.Result
   )
   Alpha=ScalarParameter() {parameterName=UseVertexColorAlphaToControlDitherMask}
  ) {constA=1}
  MaterialFunctionCall!(Mask, Scattering, Absorption) {materialFunction=/Water/Materials/Functions/WaterCoefficientMask}
  n3 = ScalarParameter() {defaultValue=0.225, group=Water Shading Model, parameterName=Specular}
  n4 = ScalarParameter() {defaultValue=0.225, group=Water Shading Model, parameterName=SpecularLowMedium}
  n5 = StaticSwitchParameter(True=QualitySwitch(Default=@n3, Low=@n4, Medium=@n4), False=@n3) {parameterName=UseSpecularLOWandMedium}
  n6 = Divide(
   A=Subtract(A=PixelDepth().R, B=ScalarParameter() {defaultValue=217.2348, parameterName=NearSpecularTransitionOffset})
   B=ScalarParameter() {defaultValue=1, parameterName=NearSpecularTransitionBlend}
  )
  n7 = StaticSwitchParameter(
   True=LinearInterpolate(A=@n5, B=ScalarParameter() {defaultValue=0.225, group=Water Shading Model, parameterName=SpecularFar}, Alpha=Saturate(None=@n6))
   False=@n5
  ) {parameterName=UseNearFarSpecularblend}
  n8 = ScalarParameter() {defaultValue=0.225, group=Water Shading Model, parameterName=SpecularMobile}
  n9 = MaterialFunctionCall!(Normal, WPO, Far Mask Override=ScalarParameter() {defaultValue=1, parameterName=FarMask}, Additional Mask, LowDetailBiome_Mask) {materialFunction=/FortWater/Materials/Functions/WaterTwoTexturePanner}
  n10 = MaterialFunctionCall!(Flowmap Velocity, Previous Foam, Normal=@n9.Normal, WPO=@n9.Normal) {materialFunction=/FortWater/Materials/Functions/WaterSimulationTexture}
  n11 = SetMaterialAttributes!(
   MaterialAttributes
   Base Color=VectorParameter() {parameterName=BaseColor}.RGB
   Specular=StaticSwitchParameter(True=ShadingPathSwitch(Default=@n7, Mobile=@n8) {desc=Migrated from FeatureLevelSwitch}, False=@n7) {parameterName=UseMobileSpecularOverride}
   Roughness=ScalarParameter() {defaultValue=0.03, group=Water Shading Model, parameterName=Roughness}
   Opacity=ScalarParameter() {group=Water Shading Model, parameterName=Surface Opacity}
   Normal=StaticSwitchParameter(True=MaterialFunctionCall() {materialFunction=/FortWater/Materials/Functions/WaterPoolVertexColorFlowMotion}.Result, False=@n10.Normal) {parameterName=UseVertexColorFlowMotion}
   World Position Offset=@n10.WPO
  ) {attributeSetTypes [
    #1
    #2
    #3
    #4
    #5
    #6
   ]}
  n12 = GetMaterialAttributes(None=@n11) {attributeGetTypes [
    #1
    #2
    #3
    #4
    #5
    #6
    #7
   ]}
  n13 = VectorParameter() {group=HLOD, parameterName=HLODColorOverride, defaultValue=<rgba 0.1612 0.3384 0.9219 1>}
  n14 = TwoSidedSign()
  n15 = Saturate(None=@n14)
  n16 = VectorParameter() {parameterName=Emissive_Night, defaultValue=<rgba 0 0 0 1>}
  n17 = MaterialFunctionCall(
   Color_Morning=VectorParameter() {parameterName=Emissive_Morning, defaultValue=<rgba 0 0 0 1>}.RGB
   Color_Day=VectorParameter() {parameterName=Emissive_Day, defaultValue=<rgba 0 0 0 1>}.RGB
   Color_Evening=@n16.RGB
   Color_Night=@n16.RGB
  ) {materialFunction=&Color_Time_of_Day}
  n18 = VectorParameter() {parameterName=EmissiveColor, defaultValue=<rgba 0 0 0 1>}
  n19 = Multiply(A=@n18.RGB, B=@n18.RGB)
  n20 = ScalarParameter() {defaultValue=1, group=Water Shading Model, parameterName=EyeAdaptionBoost}
  n21 = ScalarParameter() {defaultValue=0.85, group=HLOD, parameterName=HLODOpacity, sliderMax=1, sortPriority=33}
  n22 = PixelDepth()
  n23 = Saturate(None=Divide(A=ScalarParameter() {defaultValue=128, parameterName=Pool Depth}, B=Divide(A=@n22.R) {constB=4}))
  n24 = TwoSidedSign()
  n25 = LinearInterpolate(
   A=ScalarParameter() {defaultValue=0.1, parameterName=BottomRefraction}
   B=ScalarParameter() {defaultValue=0.15, parameterName=Refraction}
   Alpha=Saturate(None=@n24)
  )
  n26 = SetMaterialAttributes!(
   MaterialAttributes
   Base Color=MaterialProxyReplace(
    Realtime=@n12.Base Color
    MaterialProxy=StaticSwitchParameter(True=@n13.RGB, False=@n12.Base Color) {group=HLOD, parameterName=UseHLODColorOverride}
   )
   Specular=@n12.Specular
   Roughness=LinearInterpolate(A=ScalarParameter() {defaultValue=1, parameterName=RoughnessBottomSide}, B=@n12.Roughness, Alpha=@n15)
   Emissive Color=EyeAdaptationInverse(LightValueInput=StaticSwitchParameter(True=@n17.Result, False=@n19) {parameterName=Use TOD for Emissive}, AlphaInput=@n20).EyeAdaptationInverse
   Opacity=MaterialProxyReplace(Realtime=@n12.Opacity, MaterialProxy=StaticSwitchParameter(True=@n21, False=@n12.Opacity) {group=HLOD, parameterName=UseHLODColorOverride})
   Normal=Multiply(A=@n12.Normal, B=@n14)
   World Position Offset=@n12.Specular
   Refraction=Add(B=Multiply(A=@n23, B=@n25)) {constA=1}
   Opacity Mask=MaterialProxyReplace(
    Realtime=@n12.Opacity Mask
    MaterialProxy=StaticSwitchParameter(True=@n21, False=@n12.Opacity Mask) {group=HLOD, parameterName=UseHLODColorOverride}
   )
  ) {attributeSetTypes [
    #1
    #2
    #3
    #8
    #4
    #5
    #6
    #9
    #7
   ]}
  MaterialFunctionCall!(In=@n26, Velocity, Depth=Constant() {r=5000}, Transition, IsRiverMaterial, IsTranslucentMat) {materialFunction=/FortWater/Materials/WaterSurface/Hermes/Gold/MF_LiquidGold}
  Multiply(A=@n12.Specular, B=@n15)
  n27 = Saturate(None=Add(A=@n24, B=ScalarParameter() {parameterName=Min Scattering Absorption}))
  n28 = VectorParameter() {group=Water Shading Model, parameterName=Scattering, defaultValue=<rgba 0.1 0.1 0.1 0.5>}
  n29 = Divide(A=@n28.A) {constB=100}
  n30 = Constant() {r=1}
  n31 = VectorParameter() {parameterName=MobileScatterBoost, defaultValue=<rgba 1 3 5 1.5>}
  n32 = Multiply(A=@n31.RGB, B=@n31.RGB)
  n33 = Multiply(A=@n27, B=Multiply(A=Multiply(A=@n28.RGB, B=@n29), B=ShadingPathSwitch(Default=@n30, Mobile=@n32) {desc=Migrated from FeatureLevelSwitch}))
  n34 = VectorParameter() {group=Water Shading Model, parameterName=Absorption, defaultValue=<rgba 1 8 20 100>}
  n35 = Constant() {r=1}
  n36 = ScalarParameter() {defaultValue=2, parameterName=Mobile Absorption Boost, sliderMin=1}
  n37 = Divide(B=Divide(A=Multiply(A=@n34.RGB, B=@n34.RGB), B=ShadingPathSwitch(Default=@n35, Mobile=@n36) {desc=Migrated from FeatureLevelSwitch}))
  SingleLayerWaterMaterialOutput!(
   ScatteringCoefficients=EyeAdaptationInverse(LightValueInput=NamedRerouteDeclaration(None=@n33) {name=ScatteringCoef, nodeColor=<rgba 0.8665 0.5177 1 1>}, AlphaInput=@n20).EyeAdaptationInverse
   AbsorptionCoefficients=NamedRerouteDeclaration(None=Multiply(A=@n27, B=@n37)) {name=AbsorptionCoef, nodeColor=<rgba 0.4297 1 0 1>}
   PhaseG=ScalarParameter() {defaultValue=0.3, group=Water Shading Model, parameterName=Anisotropy}
   ColorScaleBehindWater
  )
 ]
}
