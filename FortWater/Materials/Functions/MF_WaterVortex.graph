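# Material function that adds a water vortex around the AthenaVortexCenter collection parameter, modifying incoming material attributes including a vertical world position offset.
# Three offset tiling grids get noise-shifted UVs and per-cell rotations relative to the center, and their samples blend to hide repetition.
# Time-driven, half-offset phases animate additional rotated texture layers.
# Distance to the center feeds the effect, and normal Z is derived from the combined XY.
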MaterialFunction /FortWater/Materials/Functions/MF_WaterVortex
graph {expressions {
  MaterialExpressionAbs_0 {class=MaterialExpressionAbs, inputs [None {from=MaterialExpressionScalarParameter_11, output=""}], properties {}}
  MaterialExpressionAbs_1 {class=MaterialExpressionAbs, inputs [None {from=MaterialExpressionScalarParameter_22, output=""}], properties {}}
  MaterialExpressionAbs_2 {class=MaterialExpressionAbs, inputs [None {from=MaterialExpressionSubtract_10, output=""}], properties {}}
  MaterialExpressionAbs_3 {class=MaterialExpressionAbs, inputs [None {from=MaterialExpressionScalarParameter_7, output=""}], properties {}}
  MaterialExpressionAbs_4 {class=MaterialExpressionAbs, inputs [None {from=MaterialExpressionSubtract_11, output=""}], properties {}}
  MaterialExpressionAdd_0 {
   class=MaterialExpressionAdd
   inputs [
    A {from=MaterialExpressionMultiply_1}
    B {from=MaterialExpressionMultiply_2}
   ]
  }
  MaterialExpressionAdd_1 {
   class=MaterialExpressionAdd
   inputs [
    A {from=MaterialExpressionConvert_0}
    B {from=MaterialExpressionMultiply_18}
   ]
  }
  MaterialExpressionAdd_10 {
   class=MaterialExpressionAdd
   inputs [
    A {from=MaterialExpressionMultiply_45}
    B {from=""}
   ]
   properties {constB=13.5}
  }
  MaterialExpressionAdd_11 {
   class=MaterialExpressionAdd
   inputs [
    A {from=MaterialExpressionRotator_2}
    B {from=MaterialExpressionFrac_8}
   ]
  }
  MaterialExpressionAdd_12 {
   class=MaterialExpressionAdd
   inputs [
    A {from=MaterialExpressionRotator_5}
    B {from=MaterialExpressionFrac_9}
   ]
  }
  MaterialExpressionAdd_13 {
   class=MaterialExpressionAdd
   inputs [
    A {from=MaterialExpressionComponentMask_5}
    B {from=MaterialExpressionDivide_2}
   ]
  }
  MaterialExpressionAdd_14 {
   class=MaterialExpressionAdd
   inputs [
    A {from=MaterialExpressionComponentMask_6}
    B {from=MaterialExpressionDivide_3}
   ]
  }
  MaterialExpressionAdd_15 {
   class=MaterialExpressionAdd
   inputs [
    A {from=MaterialExpressionComponentMask_7}
    B {from=MaterialExpressionDivide_8}
   ]
  }
  MaterialExpressionAdd_17 {
   class=MaterialExpressionAdd
   inputs [
    A {from=MaterialExpressionAdd_22}
    B {from=""}
   ]
   properties {constB=0.5}
  }
  MaterialExpressionAdd_18 {
   class=MaterialExpressionAdd
   inputs [
    A {from=MaterialExpressionMultiply_7}
    B {from=""}
   ]
   properties {constB=13.5}
  }
  MaterialExpressionAdd_19 {
   class=MaterialExpressionAdd
   inputs [
    A {from=MaterialExpressionRotator_3}
    B {from=MaterialExpressionFrac_3}
   ]
  }
  MaterialExpressionAdd_2 {
   class=MaterialExpressionAdd
   inputs [
    A {from=MaterialExpressionConvert_4}
    B {from=MaterialExpressionMultiply_21}
   ]
  }
  MaterialExpressionAdd_20 {
   class=MaterialExpressionAdd
   inputs [
    A {from=MaterialExpressionRotator_4}
    B {from=MaterialExpressionFrac_4}
   ]
  }
  MaterialExpressionAdd_21 {
   class=MaterialExpressionAdd
   inputs [
    A {from=MaterialExpressionMultiply_46}
    B {from=MaterialExpressionDivide_1}
   ]
  }
  MaterialExpressionAdd_22 {
   class=MaterialExpressionAdd
   inputs [
    A {from=MaterialExpressionMultiply_46}
    B {from=MaterialExpressionDivide_20}
   ]
  }
  MaterialExpressionAdd_3 {
   class=MaterialExpressionAdd
   inputs [
    A {from=MaterialExpressionConvert_6}
    B {from=MaterialExpressionMultiply_24}
   ]
  }
  MaterialExpressionAdd_4 {
   class=MaterialExpressionAdd
   inputs [
    A {from=MaterialExpressionMultiply_27}
    B {from=MaterialExpressionMultiply_26}
   ]
  }
  MaterialExpressionAdd_5 {
   class=MaterialExpressionAdd
   inputs [
    A {from=MaterialExpressionAdd_0}
    B {from=MaterialExpressionMultiply_3}
   ]
  }
  MaterialExpressionAdd_6 {
   class=MaterialExpressionAdd
   inputs [
    A {from=MaterialExpressionMultiply_29}
    B {from=MaterialExpressionMultiply_30}
   ]
  }
  MaterialExpressionAdd_7 {
   class=MaterialExpressionAdd
   inputs [
    A {from=MaterialExpressionMultiply_32}
    B {from=MaterialExpressionMultiply_33}
   ]
  }
  MaterialExpressionAdd_8 {
   class=MaterialExpressionAdd
   inputs [
    A {from=MaterialExpressionGetMaterialAttributes_0, output=WorldPositionOffset}
    B {from=MaterialExpressionConvert_49}
   ]
  }
  MaterialExpressionAdd_9 {
   class=MaterialExpressionAdd
   inputs [
    A {from=MaterialExpressionAdd_21}
    B {from=""}
   ]
   properties {constB=0.5}
  }
  MaterialExpressionAppendVector_0 {
   class=MaterialExpressionAppendVector
   inputs [
    A {from=MaterialExpressionNormalize_1, output=""}
    B {from=MaterialExpressionConvert_7, output=""}
   ]
  }
  MaterialExpressionAppendVector_1 {
   class=MaterialExpressionAppendVector
   inputs [
    A {from=MaterialExpressionRotator_1, output=""}
    B {from=MaterialExpressionConvert_47, output=""}
   ]
  }
  MaterialExpressionAppendVector_10 {
   class=MaterialExpressionAppendVector
   inputs [
    A {from=MaterialExpressionMultiply_48, output=""}
    B {from=MaterialExpressionConvert_20, output=""}
   ]
  }
  MaterialExpressionAppendVector_2 {
   class=MaterialExpressionAppendVector
   inputs [
    A {from=MaterialExpressionNormalize_0, output=""}
    B {from=MaterialExpressionConvert_10, output=""}
   ]
  }
  MaterialExpressionAppendVector_3 {
   class=MaterialExpressionAppendVector
   inputs [
    A {from=MaterialExpressionNormalize_2, output=""}
    B {from=MaterialExpressionConvert_13, output=""}
   ]
  }
  MaterialExpressionAppendVector_4 {
   class=MaterialExpressionAppendVector
   inputs [
    A {from=MaterialExpressionGetMaterialAttributes_0, output=BaseColor}
    B {from=MaterialExpressionGetMaterialAttributes_0, output=BaseColor}
   ]
  }
  MaterialExpressionAppendVector_5 {
   class=MaterialExpressionAppendVector
   inputs [
    A {from=MaterialExpressionRotator_0, output=""}
    B {from=MaterialExpressionConvert_46, output=""}
   ]
  }
  MaterialExpressionAppendVector_6 {
   class=MaterialExpressionAppendVector
   inputs [
    A {from=MaterialExpressionMultiply_16, output=""}
    B {from=MaterialExpressionConvert_39, output=""}
   ]
  }
  MaterialExpressionAppendVector_7 {
   class=MaterialExpressionAppendVector
   inputs [
    A {from=MaterialExpressionMultiply_12, output=""}
    B {from=MaterialExpressionConvert_48, output=""}
   ]
  }
  MaterialExpressionAppendVector_8 {
   class=MaterialExpressionAppendVector
   inputs [
    A {from=MaterialExpressionRotator_7, output=""}
    B {from=MaterialExpressionConvert_19, output=""}
   ]
  }
  MaterialExpressionAppendVector_9 {
   class=MaterialExpressionAppendVector
   inputs [
    A {from=MaterialExpressionRotator_6, output=""}
    B {from=MaterialExpressionConvert_18, output=""}
   ]
  }
  MaterialExpressionCollectionParameter_0 {
   class=MaterialExpressionCollectionParameter
   inputs []
   properties {collection=&FortniteMaterialParameters, parameterName=AthenaVortexCenter}
  }
  MaterialExpressionCollectionParameter_1 {
   class=MaterialExpressionCollectionParameter
   inputs []
   properties {collection=&FortniteMaterialParameters, parameterName=AthenaVortexCenter}
  }
  MaterialExpressionCollectionParameter_9 {
   class=MaterialExpressionCollectionParameter
   inputs []
   properties {collection=&FortniteMaterialParameters, parameterName=AthenaVortexCenter}
  }
  MaterialExpressionComponentMask_0 {class=MaterialExpressionComponentMask, inputs [None {from=MaterialExpressionVectorParameter_0, output=RGBA}], properties {a=true, b=true}}
  MaterialExpressionComponentMask_1 {class=MaterialExpressionComponentMask, inputs [None {from=MaterialExpressionVectorParameter_0, output=RGBA}], properties {g=true, r=true}}
  MaterialExpressionComponentMask_10 {class=MaterialExpressionComponentMask, inputs [None {from=MaterialExpressionMultiply_13}], properties {a=true, b=true}}
  MaterialExpressionComponentMask_11 {class=MaterialExpressionComponentMask, inputs [None {from=MaterialExpressionCollectionParameter_1}], properties {g=true, r=true}}
  MaterialExpressionComponentMask_12 {class=MaterialExpressionComponentMask, inputs [None {from=MaterialExpressionCollectionParameter_9}], properties {g=true, r=true}}
  MaterialExpressionComponentMask_13 {class=MaterialExpressionComponentMask, inputs [None {from=MaterialExpressionLinearInterpolate_11}], properties {b=true, g=true, r=true}}
  MaterialExpressionComponentMask_14 {
   class=MaterialExpressionComponentMask
   inputs [None {from=MaterialExpressionMaterialFunctionCall_1, output=Result}]
   properties {g=true, r=true}
  }
  MaterialExpressionComponentMask_15 {class=MaterialExpressionComponentMask, inputs [None {from=MaterialExpressionMaterialFunctionCall_1, output=Result}], properties {a=true}}
  MaterialExpressionComponentMask_16 {class=MaterialExpressionComponentMask, inputs [None {from=MaterialExpressionMaterialFunctionCall_2, output=Result}], properties {b=true}}
  MaterialExpressionComponentMask_17 {class=MaterialExpressionComponentMask, inputs [None {from=MaterialExpressionLinearInterpolate_11}], properties {a=true}}
  MaterialExpressionComponentMask_18 {class=MaterialExpressionComponentMask, inputs [None {from=MaterialExpressionCollectionParameter_0}], properties {g=true, r=true}}
  MaterialExpressionComponentMask_19 {
   class=MaterialExpressionComponentMask
   inputs [None {from=MaterialExpressionMaterialFunctionCall_3, output=Result}]
   properties {g=true, r=true}
  }
  MaterialExpressionComponentMask_2 {class=MaterialExpressionComponentMask, inputs [None {from=MaterialExpressionReroute_10, output=Input}], properties {g=true, r=true}}
  MaterialExpressionComponentMask_20 {class=MaterialExpressionComponentMask, inputs [None {from=MaterialExpressionMaterialFunctionCall_3, output=Result}], properties {a=true}}
  MaterialExpressionComponentMask_3 {class=MaterialExpressionComponentMask, inputs [None {from=MaterialExpressionReroute_24, output=Input}], properties {g=true, r=true}}
  MaterialExpressionComponentMask_4 {class=MaterialExpressionComponentMask, inputs [None {from=MaterialExpressionReroute_16, output=Input}], properties {g=true, r=true}}
  MaterialExpressionComponentMask_5 {class=MaterialExpressionComponentMask, inputs [None {from=MaterialExpressionVectorNoise_0}], properties {g=true, r=true}}
  MaterialExpressionComponentMask_6 {class=MaterialExpressionComponentMask, inputs [None {from=MaterialExpressionVectorNoise_1}], properties {g=true, r=true}}
  MaterialExpressionComponentMask_7 {class=MaterialExpressionComponentMask, inputs [None {from=MaterialExpressionVectorNoise_2}], properties {g=true, r=true}}
  MaterialExpressionComponentMask_8 {class=MaterialExpressionComponentMask, inputs [None {from=MaterialExpressionTextureSample_3, output=RGB}], properties {r=true}}
  MaterialExpressionComponentMask_9 {class=MaterialExpressionComponentMask, inputs [None {from=MaterialExpressionMultiply_13}], properties {g=true, r=true}}
  MaterialExpressionComposite_0 {
   class=MaterialExpressionComposite
   inputs [
    None {from=MaterialExpressionMaterialFunctionCall_4, output=Grid1}
    B {from=MaterialExpressionReroute_0, output=""}
   ]
   properties {}
  }
  MaterialExpressionComposite_2 {
   class=MaterialExpressionComposite
   inputs [
    None {from=MaterialExpressionMaterialFunctionCall_4, output=Grid2}
    B {from=MaterialExpressionReroute_0, output=""}
   ]
   properties {}
  }
  MaterialExpressionComposite_4 {
   class=MaterialExpressionComposite
   inputs [
    None {from=MaterialExpressionMaterialFunctionCall_4, output=Grid3}
    B {from=MaterialExpressionReroute_0, output=""}
   ]
   properties {}
  }
  MaterialExpressionComposite_5 {
   class=MaterialExpressionComposite
   inputs [
    None {from=MaterialExpressionCollectionParameter_1, output=""}
    Input2 {from=MaterialExpressionSubtract_0, output=""}
    B {from=MaterialExpressionReroute_4, output=Output}
   ]
   properties {}
  }
  MaterialExpressionComposite_6 {
   class=MaterialExpressionComposite
   inputs [
    None {from=MaterialExpressionCollectionParameter_1, output=""}
    Input2 {from=MaterialExpressionSubtract_2, output=""}
    B {from=MaterialExpressionReroute_6, output=Output}
   ]
   properties {}
  }
  MaterialExpressionComposite_7 {
   class=MaterialExpressionComposite
   inputs [
    None {from=MaterialExpressionCollectionParameter_1, output=""}
    Input2 {from=MaterialExpressionSubtract_3, output=""}
    B {from=MaterialExpressionReroute_9, output=Output}
   ]
   properties {}
  }
  MaterialExpressionConstantBiasScale_0 {class=MaterialExpressionConstantBiasScale, inputs [None {from=MaterialExpressionConvert_18}], properties {bias=-0.5, scale=2}}
  MaterialExpressionConstantBiasScale_1 {class=MaterialExpressionConstantBiasScale, inputs [None {from=MaterialExpressionConvert_19}], properties {bias=-0.5, scale=2}}
  MaterialExpressionConstantBiasScale_10 {class=MaterialExpressionConstantBiasScale, inputs [None {from=MaterialExpressionConvert_46}], properties {bias=-0.5, scale=2}}
  MaterialExpressionConstantBiasScale_11 {class=MaterialExpressionConstantBiasScale, inputs [None {from=MaterialExpressionConvert_47}], properties {bias=-0.5, scale=2}}
  MaterialExpressionConstantBiasScale_5 {class=MaterialExpressionConstantBiasScale, inputs [None {from=MaterialExpressionConvert_28}], properties {bias=-0.5, scale=2}}
  MaterialExpressionConstantBiasScale_8 {class=MaterialExpressionConstantBiasScale, inputs [None {from=MaterialExpressionConvert_40}], properties {bias=-0.5, scale=2}}
  MaterialExpressionConstantBiasScale_9 {class=MaterialExpressionConstantBiasScale, inputs [None {from=MaterialExpressionConvert_43}], properties {bias=-0.5, scale=2}}
  MaterialExpressionConstant_0 {class=MaterialExpressionConstant, properties {r=1}}
  MaterialExpressionConstant_1 {class=MaterialExpressionConstant, properties {r=1.732}}
  MaterialExpressionConstant_2 {class=MaterialExpressionConstant}
  MaterialExpressionConvert_0 {
   class=MaterialExpressionConvert
   inputs [None {from=MaterialExpressionReroute_1, output=Input}]
   properties {
    nodeName=Break Float2
    convertInputs [{
      type=Vector2
      expressionInput {
       expression=&self:MaterialExpressionReroute_1
       inputName=None
       mask=0
       maskA=0
       maskB=0
       maskG=0
       maskR=0
       outputIndex=0
      }
     }]
    convertOutputs [{type=Scalar}, {type=Scalar}]
   }
  }
  MaterialExpressionConvert_1 {
   class=MaterialExpressionConvert
   inputs [
    None {from=MaterialExpressionAdd_1}
    None {from=MaterialExpressionMultiply_19}
   ]
   properties {
    nodeName=Make Vector2
    convertInputs [
     {
      type=Scalar
      expressionInput {
       expression=&self:MaterialExpressionAdd_1
       inputName=None
       mask=0
       maskA=0
       maskB=0
       maskG=0
       maskR=0
       outputIndex=0
      }
     }
     {
      type=Scalar
      expressionInput {
       expression=&self:MaterialExpressionMultiply_19
       inputName=None
       mask=0
       maskA=0
       maskB=0
       maskG=0
       maskR=0
       outputIndex=0
      }
     }
    ]
    convertOutputs [{type=Vector2}]
   }
  }
  MaterialExpressionConvert_10 {
   class=MaterialExpressionConvert
   inputs [None {from=MaterialExpressionMultiply_31}]
   properties {
    nodeName=Swizzle
    convertInputs [{
      type=Vector2
      expressionInput {
       expression=&self:MaterialExpressionMultiply_31
       inputName=None
       mask=0
       maskA=0
       maskB=0
       maskG=0
       maskR=0
       outputIndex=0
      }
     }]
    convertOutputs [{type=Vector2}]
   }
  }
  MaterialExpressionConvert_11 {
   class=MaterialExpressionConvert
   inputs [None {from=MaterialExpressionReroute_25, output=Input2}]
   properties {
    nodeName=Break Float2
    convertInputs [{
      type=Vector2
      expressionInput {
       expression=&self:MaterialExpressionReroute_25
       inputName=None
       mask=0
       maskA=0
       maskB=0
       maskG=0
       maskR=0
       outputIndex=0
      }
     }]
    convertOutputs [{type=Scalar}, {type=Scalar}]
   }
  }
  MaterialExpressionConvert_12 {
   class=MaterialExpressionConvert
   inputs [
    None {from=""}
    None {from=""}
   ]
   properties {
    nodeName=Make Vector2
    convertInputs [
     {
      type=Scalar
      defaultValue=<rgba 1 0 0 1>
      expressionInput {expression=null, inputName=None, mask=0, maskA=0, maskB=0, maskG=0, maskR=0, outputIndex=0}
     }
     {
      type=Scalar
      defaultValue=<rgba -1 0 0 1>
      expressionInput {expression=null, inputName=None, mask=0, maskA=0, maskB=0, maskG=0, maskR=0, outputIndex=0}
     }
    ]
    convertOutputs [{type=Vector2}]
   }
  }
  MaterialExpressionConvert_13 {
   class=MaterialExpressionConvert
   inputs [None {from=MaterialExpressionMultiply_34}]
   properties {
    nodeName=Swizzle
    convertInputs [{
      type=Vector2
      expressionInput {
       expression=&self:MaterialExpressionMultiply_34
       inputName=None
       mask=0
       maskA=0
       maskB=0
       maskG=0
       maskR=0
       outputIndex=0
      }
     }]
    convertOutputs [{type=Vector2}]
   }
  }
  MaterialExpressionConvert_14 {
   class=MaterialExpressionConvert
   inputs [None {from=MaterialExpressionReroute_17, output=Input2}]
   properties {
    nodeName=Break Float2
    convertInputs [{
      type=Vector2
      expressionInput {
       expression=&self:MaterialExpressionReroute_17
       inputName=None
       mask=0
       maskA=0
       maskB=0
       maskG=0
       maskR=0
       outputIndex=0
      }
     }]
    convertOutputs [{type=Scalar}, {type=Scalar}]
   }
  }
  MaterialExpressionConvert_15 {
   class=MaterialExpressionConvert
   inputs [
    None {from=""}
    None {from=""}
   ]
   properties {
    nodeName=Make Vector2
    convertInputs [
     {
      type=Scalar
      defaultValue=<rgba 1 0 0 1>
      expressionInput {expression=null, inputName=None, mask=0, maskA=0, maskB=0, maskG=0, maskR=0, outputIndex=0}
     }
     {
      type=Scalar
      defaultValue=<rgba -1 0 0 1>
      expressionInput {expression=null, inputName=None, mask=0, maskA=0, maskB=0, maskG=0, maskR=0, outputIndex=0}
     }
    ]
    convertOutputs [{type=Vector2}]
   }
  }
  MaterialExpressionConvert_16 {
   class=MaterialExpressionConvert
   inputs [
    None {from=""}
    None {from=""}
   ]
   properties {
    nodeName=Make Vector2
    convertInputs [
     {
      type=Scalar
      defaultValue=<rgba 0.013 0 0 1>
      expressionInput {expression=null, inputName=None, mask=0, maskA=0, maskB=0, maskG=0, maskR=0, outputIndex=0}
     }
     {
      type=Scalar
      defaultValue=<rgba 0.057 0 0 1>
      expressionInput {expression=null, inputName=None, mask=0, maskA=0, maskB=0, maskG=0, maskR=0, outputIndex=0}
     }
    ]
    convertOutputs [{type=Vector2}]
   }
  }
  MaterialExpressionConvert_17 {
   class=MaterialExpressionConvert
   inputs [
    None {from=""}
    None {from=""}
   ]
   properties {
    nodeName=Make Vector2
    convertInputs [
     {
      type=Scalar
      defaultValue=<rgba 0.013 0 0 1>
      expressionInput {expression=null, inputName=None, mask=0, maskA=0, maskB=0, maskG=0, maskR=0, outputIndex=0}
     }
     {
      type=Scalar
      defaultValue=<rgba 0.057 0 0 1>
      expressionInput {expression=null, inputName=None, mask=0, maskA=0, maskB=0, maskG=0, maskR=0, outputIndex=0}
     }
    ]
    convertOutputs [{type=Vector2}]
   }
  }
  MaterialExpressionConvert_18 {
   class=MaterialExpressionConvert
   inputs [None {from=MaterialExpressionTextureSample_6, output=RGBA}]
   properties {
    nodeName=Break Float3
    convertInputs [{
      type=Vector4
      defaultValue=<rgba 0 0 0 1>
      expressionInput {
       expression=&self:MaterialExpressionTextureSample_6
       inputName=None
       mask=1
       maskA=1
       maskB=1
       maskG=1
       maskR=1
       outputIndex=5
      }
     }]
    convertOutputs [{type=Vector2}, {type=Vector2}]
   }
  }
  MaterialExpressionConvert_19 {
   class=MaterialExpressionConvert
   inputs [None {from=MaterialExpressionTextureSample_7, output=RGBA}]
   properties {
    nodeName=Break Float3
    convertInputs [{
      type=Vector4
      defaultValue=<rgba 0 0 0 1>
      expressionInput {
       expression=&self:MaterialExpressionTextureSample_7
       inputName=None
       mask=1
       maskA=1
       maskB=1
       maskG=1
       maskR=1
       outputIndex=5
      }
     }]
    convertOutputs [{type=Vector2}, {type=Vector2}]
   }
  }
  MaterialExpressionConvert_2 {
   class=MaterialExpressionConvert
   inputs [None {from=MaterialExpressionNamedRerouteUsage_0}]
   properties {
    nodeName=Break Float3
    convertInputs [{
      type=Vector3
      expressionInput {
       expression=&self:MaterialExpressionNamedRerouteUsage_0
       inputName=None
       mask=0
       maskA=0
       maskB=0
       maskG=0
       maskR=0
       outputIndex=0
      }
     }]
    convertOutputs [{type=Scalar}, {type=Scalar}, {type=Scalar}]
   }
  }
  MaterialExpressionConvert_20 {
   class=MaterialExpressionConvert
   inputs [None {from=MaterialExpressionLinearInterpolate_16}]
   properties {
    nodeName=Break Float3
    convertInputs [{
      type=Vector4
      defaultValue=<rgba 0 0 0 1>
      expressionInput {
       expression=&self:MaterialExpressionLinearInterpolate_16
       inputName=None
       mask=0
       maskA=0
       maskB=0
       maskG=0
       maskR=0
       outputIndex=0
      }
     }]
    convertOutputs [{type=Vector2}, {type=Vector2}]
   }
  }
  MaterialExpressionConvert_25 {
   class=MaterialExpressionConvert
   inputs [
    None {from=MaterialExpressionMaterialFunctionCall_4, output=Grid1}
    None {from=""}
   ]
   properties {
    nodeName=Make Vector3
    convertInputs [
     {
      type=Vector2
      expressionInput {
       expression=&self:MaterialExpressionMaterialFunctionCall_4
       inputName=None
       mask=0
       maskA=0
       maskB=0
       maskG=0
       maskR=0
       outputIndex=0
      }
     }
     {type=Scalar, expressionInput {expression=null, inputName=None, mask=0, maskA=0, maskB=0, maskG=0, maskR=0, outputIndex=0}}
    ]
    convertOutputs [{type=Vector3}]
   }
  }
  MaterialExpressionConvert_26 {
   class=MaterialExpressionConvert
   inputs [
    None {from=MaterialExpressionMaterialFunctionCall_4, output=Grid2}
    None {from=""}
   ]
   properties {
    nodeName=Make Vector3
    convertInputs [
     {
      type=Vector2
      expressionInput {
       expression=&self:MaterialExpressionMaterialFunctionCall_4
       inputName=None
       mask=0
       maskA=0
       maskB=0
       maskG=0
       maskR=0
       outputIndex=1
      }
     }
     {type=Scalar, expressionInput {expression=null, inputName=None, mask=0, maskA=0, maskB=0, maskG=0, maskR=0, outputIndex=0}}
    ]
    convertOutputs [{type=Vector3}]
   }
  }
  MaterialExpressionConvert_27 {
   class=MaterialExpressionConvert
   inputs [
    None {from=MaterialExpressionMaterialFunctionCall_4, output=Grid3}
    None {from=""}
   ]
   properties {
    nodeName=Make Vector3
    convertInputs [
     {
      type=Vector2
      expressionInput {
       expression=&self:MaterialExpressionMaterialFunctionCall_4
       inputName=None
       mask=0
       maskA=0
       maskB=0
       maskG=0
       maskR=0
       outputIndex=2
      }
     }
     {type=Scalar, expressionInput {expression=null, inputName=None, mask=0, maskA=0, maskB=0, maskG=0, maskR=0, outputIndex=0}}
    ]
    convertOutputs [{type=Vector3}]
   }
  }
  MaterialExpressionConvert_28 {
   class=MaterialExpressionConvert
   inputs [None {from=MaterialExpressionTextureSample_0, output=RGBA}]
   properties {
    nodeName=Break Float3
    convertInputs [{
      type=Vector4
      defaultValue=<rgba 0 0 0 1>
      expressionInput {
       expression=&self:MaterialExpressionTextureSample_0
       inputName=None
       mask=1
       maskA=1
       maskB=1
       maskG=1
       maskR=1
       outputIndex=5
      }
     }]
    convertOutputs [{type=Vector2}, {type=Vector2}]
   }
  }
  MaterialExpressionConvert_29 {
   class=MaterialExpressionConvert
   inputs [None {from=MaterialExpressionNamedRerouteUsage_3}]
   properties {
    nodeName=Break Float2
    convertInputs [{
      type=Vector4
      defaultValue=<rgba 0 0 0 1>
      expressionInput {
       expression=&self:MaterialExpressionNamedRerouteUsage_3
       inputName=None
       mask=0
       maskA=0
       maskB=0
       maskG=0
       maskR=0
       outputIndex=0
      }
     }]
    convertOutputs [{type=Vector2}, {type=Vector2}]
   }
  }
  MaterialExpressionConvert_3 {
   class=MaterialExpressionConvert
   inputs [
    None {from=MaterialExpressionAdd_2}
    None {from=MaterialExpressionMultiply_22}
   ]
   properties {
    nodeName=Make Vector2
    convertInputs [
     {
      type=Scalar
      expressionInput {
       expression=&self:MaterialExpressionAdd_2
       inputName=None
       mask=0
       maskA=0
       maskB=0
       maskG=0
       maskR=0
       outputIndex=0
      }
     }
     {
      type=Scalar
      expressionInput {
       expression=&self:MaterialExpressionMultiply_22
       inputName=None
       mask=0
       maskA=0
       maskB=0
       maskG=0
       maskR=0
       outputIndex=0
      }
     }
    ]
    convertOutputs [{type=Vector2}]
   }
  }
  MaterialExpressionConvert_30 {
   class=MaterialExpressionConvert
   inputs [
    None {from=MaterialExpressionDotProduct_1}
    None {from=MaterialExpressionDotProduct_3}
    None {from=MaterialExpressionConvert_28}
   ]
   properties {
    nodeName=Make Vector3
    convertInputs [
     {
      type=Scalar
      expressionInput {
       expression=&self:MaterialExpressionDotProduct_1
       inputName=None
       mask=0
       maskA=0
       maskB=0
       maskG=0
       maskR=0
       outputIndex=0
      }
     }
     {
      type=Scalar
      expressionInput {
       expression=&self:MaterialExpressionDotProduct_3
       inputName=None
       mask=0
       maskA=0
       maskB=0
       maskG=0
       maskR=0
       outputIndex=0
      }
     }
     {
      type=Vector2
      expressionInput {
       expression=&self:MaterialExpressionConvert_28
       inputName=None
       mask=0
       maskA=0
       maskB=0
       maskG=0
       maskR=0
       outputIndex=1
      }
     }
    ]
    convertOutputs [{type=Vector4, defaultValue=<rgba 0 0 0 1>}]
   }
  }
  MaterialExpressionConvert_37 {
   class=MaterialExpressionConvert
   inputs [
    None {from=""}
    None {from=""}
   ]
   properties {
    nodeName=Make Vector2
    convertInputs [
     {
      type=Scalar
      defaultValue=<rgba 0.013 0 0 1>
      expressionInput {expression=null, inputName=None, mask=0, maskA=0, maskB=0, maskG=0, maskR=0, outputIndex=0}
     }
     {
      type=Scalar
      defaultValue=<rgba 0.057 0 0 1>
      expressionInput {expression=null, inputName=None, mask=0, maskA=0, maskB=0, maskG=0, maskR=0, outputIndex=0}
     }
    ]
    convertOutputs [{type=Vector2}]
   }
  }
  MaterialExpressionConvert_38 {
   class=MaterialExpressionConvert
   inputs [
    None {from=""}
    None {from=""}
   ]
   properties {
    nodeName=Make Vector2
    convertInputs [
     {
      type=Scalar
      defaultValue=<rgba 0.013 0 0 1>
      expressionInput {expression=null, inputName=None, mask=0, maskA=0, maskB=0, maskG=0, maskR=0, outputIndex=0}
     }
     {
      type=Scalar
      defaultValue=<rgba 0.057 0 0 1>
      expressionInput {expression=null, inputName=None, mask=0, maskA=0, maskB=0, maskG=0, maskR=0, outputIndex=0}
     }
    ]
    convertOutputs [{type=Vector2}]
   }
  }
  MaterialExpressionConvert_39 {
   class=MaterialExpressionConvert
   inputs [None {from=MaterialExpressionAdd_5}]
   properties {
    nodeName=Break Float3
    convertInputs [{
      type=Vector4
      defaultValue=<rgba 0 0 0 1>
      expressionInput {
       expression=&self:MaterialExpressionAdd_5
       inputName=None
       mask=0
       maskA=0
       maskB=0
       maskG=0
       maskR=0
       outputIndex=0
      }
     }]
    convertOutputs [{type=Vector2}, {type=Vector2}]
   }
  }
  MaterialExpressionConvert_4 {
   class=MaterialExpressionConvert
   inputs [None {from=MaterialExpressionReroute_2, output=Input}]
   properties {
    nodeName=Break Float2
    convertInputs [{
      type=Vector2
      expressionInput {
       expression=&self:MaterialExpressionReroute_2
       inputName=None
       mask=0
       maskA=0
       maskB=0
       maskG=0
       maskR=0
       outputIndex=0
      }
     }]
    convertOutputs [{type=Scalar}, {type=Scalar}]
   }
  }
  MaterialExpressionConvert_40 {
   class=MaterialExpressionConvert
   inputs [None {from=MaterialExpressionTextureSample_1, output=RGBA}]
   properties {
    nodeName=Break Float3
    convertInputs [{
      type=Vector4
      defaultValue=<rgba 0 0 0 1>
      expressionInput {
       expression=&self:MaterialExpressionTextureSample_1
       inputName=None
       mask=1
       maskA=1
       maskB=1
       maskG=1
       maskR=1
       outputIndex=5
      }
     }]
    convertOutputs [{type=Vector2}, {type=Vector2}]
   }
  }
  MaterialExpressionConvert_41 {
   class=MaterialExpressionConvert
   inputs [None {from=MaterialExpressionNamedRerouteUsage_4}]
   properties {
    nodeName=Break Float2
    convertInputs [{
      type=Vector4
      defaultValue=<rgba 0 0 0 1>
      expressionInput {
       expression=&self:MaterialExpressionNamedRerouteUsage_4
       inputName=None
       mask=0
       maskA=0
       maskB=0
       maskG=0
       maskR=0
       outputIndex=0
      }
     }]
    convertOutputs [{type=Vector2}, {type=Vector2}]
   }
  }
  MaterialExpressionConvert_42 {
   class=MaterialExpressionConvert
   inputs [
    None {from=MaterialExpressionDotProduct_12}
    None {from=MaterialExpressionDotProduct_11}
    None {from=MaterialExpressionConvert_40}
   ]
   properties {
    nodeName=Make Vector3
    convertInputs [
     {
      type=Scalar
      expressionInput {
       expression=&self:MaterialExpressionDotProduct_12
       inputName=None
       mask=0
       maskA=0
       maskB=0
       maskG=0
       maskR=0
       outputIndex=0
      }
     }
     {
      type=Scalar
      expressionInput {
       expression=&self:MaterialExpressionDotProduct_11
       inputName=None
       mask=0
       maskA=0
       maskB=0
       maskG=0
       maskR=0
       outputIndex=0
      }
     }
     {
      type=Vector2
      expressionInput {
       expression=&self:MaterialExpressionConvert_40
       inputName=None
       mask=0
       maskA=0
       maskB=0
       maskG=0
       maskR=0
       outputIndex=1
      }
     }
    ]
    convertOutputs [{type=Vector4, defaultValue=<rgba 0 0 0 1>}]
   }
  }
  MaterialExpressionConvert_43 {
   class=MaterialExpressionConvert
   inputs [None {from=MaterialExpressionTextureSample_2, output=RGBA}]
   properties {
    nodeName=Break Float3
    convertInputs [{
      type=Vector4
      defaultValue=<rgba 0 0 0 1>
      expressionInput {
       expression=&self:MaterialExpressionTextureSample_2
       inputName=None
       mask=1
       maskA=1
       maskB=1
       maskG=1
       maskR=1
       outputIndex=5
      }
     }]
    convertOutputs [{type=Vector2}, {type=Vector2}]
   }
  }
  MaterialExpressionConvert_44 {
   class=MaterialExpressionConvert
   inputs [None {from=MaterialExpressionNamedRerouteUsage_5}]
   properties {
    nodeName=Break Float2
    convertInputs [{
      type=Vector4
      defaultValue=<rgba 0 0 0 1>
      expressionInput {
       expression=&self:MaterialExpressionNamedRerouteUsage_5
       inputName=None
       mask=0
       maskA=0
       maskB=0
       maskG=0
       maskR=0
       outputIndex=0
      }
     }]
    convertOutputs [{type=Vector2}, {type=Vector2}]
   }
  }
  MaterialExpressionConvert_45 {
   class=MaterialExpressionConvert
   inputs [
    None {from=MaterialExpressionDotProduct_13}
    None {from=MaterialExpressionDotProduct_14}
    None {from=MaterialExpressionConvert_43}
   ]
   properties {
    nodeName=Make Vector3
    convertInputs [
     {
      type=Scalar
      expressionInput {
       expression=&self:MaterialExpressionDotProduct_13
       inputName=None
       mask=0
       maskA=0
       maskB=0
       maskG=0
       maskR=0
       outputIndex=0
      }
     }
     {
      type=Scalar
      expressionInput {
       expression=&self:MaterialExpressionDotProduct_14
       inputName=None
       mask=0
       maskA=0
       maskB=0
       maskG=0
       maskR=0
       outputIndex=0
      }
     }
     {
      type=Vector2
      expressionInput {
       expression=&self:MaterialExpressionConvert_43
       inputName=None
       mask=0
       maskA=0
       maskB=0
       maskG=0
       maskR=0
       outputIndex=1
      }
     }
    ]
    convertOutputs [{type=Vector4, defaultValue=<rgba 0 0 0 1>}]
   }
  }
  MaterialExpressionConvert_46 {
   class=MaterialExpressionConvert
   inputs [None {from=MaterialExpressionTextureSample_4, output=RGBA}]
   properties {
    nodeName=Break Float3
    convertInputs [{
      type=Vector4
      defaultValue=<rgba 0 0 0 1>
      expressionInput {
       expression=&self:MaterialExpressionTextureSample_4
       inputName=None
       mask=1
       maskA=1
       maskB=1
       maskG=1
       maskR=1
       outputIndex=5
      }
     }]
    convertOutputs [{type=Vector2}, {type=Vector2}]
   }
  }
  MaterialExpressionConvert_47 {
   class=MaterialExpressionConvert
   inputs [None {from=MaterialExpressionTextureSample_5, output=RGBA}]
   properties {
    nodeName=Break Float3
    convertInputs [{
      type=Vector4
      defaultValue=<rgba 0 0 0 1>
      expressionInput {
       expression=&self:MaterialExpressionTextureSample_5
       inputName=None
       mask=1
       maskA=1
       maskB=1
       maskG=1
       maskR=1
       outputIndex=5
      }
     }]
    convertOutputs [{type=Vector2}, {type=Vector2}]
   }
  }
  MaterialExpressionConvert_48 {
   class=MaterialExpressionConvert
   inputs [None {from=MaterialExpressionLinearInterpolate_3}]
   properties {
    nodeName=Break Float3
    convertInputs [{
      type=Vector4
      defaultValue=<rgba 0 0 0 1>
      expressionInput {
       expression=&self:MaterialExpressionLinearInterpolate_3
       inputName=None
       mask=0
       maskA=0
       maskB=0
       maskG=0
       maskR=0
       outputIndex=0
      }
     }]
    convertOutputs [{type=Vector2}, {type=Vector2}]
   }
  }
  MaterialExpressionConvert_49 {
   class=MaterialExpressionConvert
   inputs [
    None {from=""}
    None {from=""}
    None {from=MaterialExpressionMultiply_14}
   ]
   properties {
    nodeName=Make Vector3
    convertInputs [
     {type=Scalar, expressionInput {expression=null, inputName=None, mask=0, maskA=0, maskB=0, maskG=0, maskR=0, outputIndex=0}}
     {type=Scalar, expressionInput {expression=null, inputName=None, mask=0, maskA=0, maskB=0, maskG=0, maskR=0, outputIndex=0}}
     {
      type=Scalar
      expressionInput {
       expression=&self:MaterialExpressionMultiply_14
       inputName=None
       mask=0
       maskA=0
       maskB=0
       maskG=0
       maskR=0
       outputIndex=0
      }
     }
    ]
    convertOutputs [{type=Vector3}]
   }
  }
  MaterialExpressionConvert_5 {
   class=MaterialExpressionConvert
   inputs [
    None {from=MaterialExpressionAdd_3}
    None {from=MaterialExpressionMultiply_25}
   ]
   properties {
    nodeName=Make Vector2
    convertInputs [
     {
      type=Scalar
      expressionInput {
       expression=&self:MaterialExpressionAdd_3
       inputName=None
       mask=0
       maskA=0
       maskB=0
       maskG=0
       maskR=0
       outputIndex=0
      }
     }
     {
      type=Scalar
      expressionInput {
       expression=&self:MaterialExpressionMultiply_25
       inputName=None
       mask=0
       maskA=0
       maskB=0
       maskG=0
       maskR=0
       outputIndex=0
      }
     }
    ]
    convertOutputs [{type=Vector2}]
   }
  }
  MaterialExpressionConvert_6 {
   class=MaterialExpressionConvert
   inputs [None {from=MaterialExpressionReroute_7, output=Input}]
   properties {
    nodeName=Break Float2
    convertInputs [{
      type=Vector2
      expressionInput {
       expression=&self:MaterialExpressionReroute_7
       inputName=None
       mask=0
       maskA=0
       maskB=0
       maskG=0
       maskR=0
       outputIndex=0
      }
     }]
    convertOutputs [{type=Scalar}, {type=Scalar}]
   }
  }
  MaterialExpressionConvert_7 {
   class=MaterialExpressionConvert
   inputs [None {from=MaterialExpressionMultiply_28}]
   properties {
    nodeName=Swizzle
    convertInputs [{
      type=Vector2
      expressionInput {
       expression=&self:MaterialExpressionMultiply_28
       inputName=None
       mask=0
       maskA=0
       maskB=0
       maskG=0
       maskR=0
       outputIndex=0
      }
     }]
    convertOutputs [{type=Vector2}]
   }
  }
  MaterialExpressionConvert_8 {
   class=MaterialExpressionConvert
   inputs [None {from=MaterialExpressionReroute_11, output=Input2}]
   properties {
    nodeName=Break Float2
    convertInputs [{
      type=Vector2
      expressionInput {
       expression=&self:MaterialExpressionReroute_11
       inputName=None
       mask=0
       maskA=0
       maskB=0
       maskG=0
       maskR=0
       outputIndex=0
      }
     }]
    convertOutputs [{type=Scalar}, {type=Scalar}]
   }
  }
  MaterialExpressionConvert_9 {
   class=MaterialExpressionConvert
   inputs [
    None {from=""}
    None {from=""}
   ]
   properties {
    nodeName=Make Vector2
    convertInputs [
     {
      type=Scalar
      defaultValue=<rgba 1 0 0 1>
      expressionInput {expression=null, inputName=None, mask=0, maskA=0, maskB=0, maskG=0, maskR=0, outputIndex=0}
     }
     {
      type=Scalar
      defaultValue=<rgba -1 0 0 1>
      expressionInput {expression=null, inputName=None, mask=0, maskA=0, maskB=0, maskG=0, maskR=0, outputIndex=0}
     }
    ]
    convertOutputs [{type=Vector2}]
   }
  }
  MaterialExpressionDeriveNormalZ_0 {class=MaterialExpressionDeriveNormalZ, inputs [InXY {from=MaterialExpressionMultiply_50, output=""}], properties {}}
  MaterialExpressionDeriveNormalZ_5 {class=MaterialExpressionDeriveNormalZ, inputs [InXY {from=MaterialExpressionComponentMask_14, output=""}], properties {}}
  MaterialExpressionDistance_0 {
   class=MaterialExpressionDistance
   inputs [
    A {from=MaterialExpressionComponentMask_11, output=""}
    B {from=MaterialExpressionReroute_4, output=Output}
   ]
   properties {}
  }
  MaterialExpressionDistance_1 {
   class=MaterialExpressionDistance
   inputs [
    A {from=MaterialExpressionComponentMask_11, output=""}
    B {from=MaterialExpressionReroute_6, output=Output}
   ]
   properties {}
  }
  MaterialExpressionDistance_2 {
   class=MaterialExpressionDistance
   inputs [
    A {from=MaterialExpressionComponentMask_11, output=""}
    B {from=MaterialExpressionReroute_9, output=Output}
   ]
   properties {}
  }
  MaterialExpressionDistance_3 {
   class=MaterialExpressionDistance
   inputs [
    A {from=MaterialExpressionWorldPosition_3, output=XY}
    B {from=MaterialExpressionComponentMask_11, output=""}
   ]
   properties {}
  }
  MaterialExpressionDistance_4 {
   class=MaterialExpressionDistance
   inputs [
    A {from=MaterialExpressionComponentMask_12, output=""}
    B {from=MaterialExpressionWorldPosition_4, output=XY}
   ]
   properties {}
  }
  MaterialExpressionDivide_0 {
   class=MaterialExpressionDivide
   inputs [
    A {from=MaterialExpressionSubtract_8}
    B {from=MaterialExpressionMultiply_36}
   ]
  }
  MaterialExpressionDivide_1 {
   class=MaterialExpressionDivide
   inputs [
    A {from=MaterialExpressionTime_0}
    B {from=MaterialExpressionScalarParameter_10}
   ]
  }
  MaterialExpressionDivide_10 {
   class=MaterialExpressionDivide
   inputs [
    A {from=MaterialExpressionConstant_1}
    B {from=""}
   ]
  }
  MaterialExpressionDivide_12 {
   class=MaterialExpressionDivide
   inputs [
    A {from=MaterialExpressionWorldPosition_2, output=XY}
    B {from=MaterialExpressionScalarParameter_17}
   ]
  }
  MaterialExpressionDivide_14 {
   class=MaterialExpressionDivide
   inputs [
    A {from=MaterialExpressionDistance_0}
    B {from=MaterialExpressionScalarParameter_20}
   ]
  }
  MaterialExpressionDivide_15 {
   class=MaterialExpressionDivide
   inputs [
    A {from=MaterialExpressionDistance_1}
    B {from=MaterialExpressionScalarParameter_20}
   ]
  }
  MaterialExpressionDivide_16 {
   class=MaterialExpressionDivide
   inputs [
    A {from=MaterialExpressionDistance_2}
    B {from=MaterialExpressionScalarParameter_20}
   ]
  }
  MaterialExpressionDivide_18 {
   class=MaterialExpressionDivide
   inputs [
    A {from=MaterialExpressionSubtract_9}
    B {from=MaterialExpressionMultiply_51}
   ]
  }
  MaterialExpressionDivide_2 {
   class=MaterialExpressionDivide
   inputs [
    A {from=MaterialExpressionReroute_13, output=Output}
    B {from=MaterialExpressionMultiply_47}
   ]
  }
  MaterialExpressionDivide_20 {
   class=MaterialExpressionDivide
   inputs [
    A {from=MaterialExpressionTime_3}
    B {from=MaterialExpressionScalarParameter_25}
   ]
  }
  MaterialExpressionDivide_3 {
   class=MaterialExpressionDivide
   inputs [
    A {from=MaterialExpressionReroute_27, output=Output}
    B {from=MaterialExpressionMultiply_47}
   ]
  }
  MaterialExpressionDivide_8 {
   class=MaterialExpressionDivide
   inputs [
    A {from=MaterialExpressionReroute_23, output=Output}
    B {from=MaterialExpressionMultiply_47}
   ]
  }
  MaterialExpressionDotProduct_0 {
   class=MaterialExpressionDotProduct
   inputs [
    A {from=MaterialExpressionDivide_0, output=""}
    B {from=MaterialExpressionDivide_0, output=""}
   ]
   properties {}
  }
  MaterialExpressionDotProduct_1 {
   class=MaterialExpressionDotProduct
   inputs [
    A {from=MaterialExpressionConvert_29, output=""}
    B {from=MaterialExpressionConstantBiasScale_5, output=""}
   ]
   properties {}
  }
  MaterialExpressionDotProduct_11 {
   class=MaterialExpressionDotProduct
   inputs [
    A {from=MaterialExpressionConvert_41, output=""}
    B {from=MaterialExpressionConstantBiasScale_8, output=""}
   ]
   properties {}
  }
  MaterialExpressionDotProduct_12 {
   class=MaterialExpressionDotProduct
   inputs [
    A {from=MaterialExpressionConvert_41, output=""}
    B {from=MaterialExpressionConstantBiasScale_8, output=""}
   ]
   properties {}
  }
  MaterialExpressionDotProduct_13 {
   class=MaterialExpressionDotProduct
   inputs [
    A {from=MaterialExpressionConvert_44, output=""}
    B {from=MaterialExpressionConstantBiasScale_9, output=""}
   ]
   properties {}
  }
  MaterialExpressionDotProduct_14 {
   class=MaterialExpressionDotProduct
   inputs [
    A {from=MaterialExpressionConvert_44, output=""}
    B {from=MaterialExpressionConstantBiasScale_9, output=""}
   ]
   properties {}
  }
  MaterialExpressionDotProduct_3 {
   class=MaterialExpressionDotProduct
   inputs [
    A {from=MaterialExpressionConvert_29, output=""}
    B {from=MaterialExpressionConstantBiasScale_5, output=""}
   ]
   properties {}
  }
  MaterialExpressionExponential_0 {class=MaterialExpressionExponential, inputs [None {from=MaterialExpressionLinearInterpolate_2, output=""}], properties {}}
  MaterialExpressionExponential_1 {class=MaterialExpressionExponential, inputs [None {from=MaterialExpressionLinearInterpolate_1, output=""}], properties {}}
  MaterialExpressionExponential_2 {class=MaterialExpressionExponential, inputs [None {from=MaterialExpressionLinearInterpolate_15, output=""}], properties {}}
  MaterialExpressionExponential_3 {class=MaterialExpressionExponential, inputs [None {from=MaterialExpressionLinearInterpolate_14, output=""}], properties {}}
  MaterialExpressionFloor_0 {class=MaterialExpressionFloor, inputs [None {from=MaterialExpressionAdd_22, output=""}], properties {}}
  MaterialExpressionFloor_1 {class=MaterialExpressionFloor, inputs [None {from=MaterialExpressionAdd_17, output=""}], properties {}}
  MaterialExpressionFloor_2 {class=MaterialExpressionFloor, inputs [None {from=MaterialExpressionAdd_21, output=""}], properties {}}
  MaterialExpressionFloor_3 {class=MaterialExpressionFloor, inputs [None {from=MaterialExpressionAdd_9, output=""}], properties {}}
  MaterialExpressionFrac_0 {class=MaterialExpressionFrac, inputs [None {from=MaterialExpressionAdd_22, output=""}], properties {}}
  MaterialExpressionFrac_1 {class=MaterialExpressionFrac, inputs [None {from=MaterialExpressionAdd_17, output=""}], properties {}}
  MaterialExpressionFrac_3 {class=MaterialExpressionFrac, inputs [None {from=MaterialExpressionMultiply_56, output=""}], properties {}}
  MaterialExpressionFrac_4 {class=MaterialExpressionFrac, inputs [None {from=MaterialExpressionMultiply_57, output=""}], properties {}}
  MaterialExpressionFrac_5 {class=MaterialExpressionFrac, inputs [None {from=MaterialExpressionAdd_21, output=""}], properties {}}
  MaterialExpressionFrac_6 {class=MaterialExpressionFrac, inputs [None {from=MaterialExpressionAdd_9, output=""}], properties {}}
  MaterialExpressionFrac_8 {class=MaterialExpressionFrac, inputs [None {from=MaterialExpressionMultiply_40, output=""}], properties {}}
  MaterialExpressionFrac_9 {class=MaterialExpressionFrac, inputs [None {from=MaterialExpressionMultiply_41, output=""}], properties {}}
  MaterialExpressionFunctionInput_1 {
   class=MaterialExpressionFunctionInput
   inputs [Preview {from="", output=""}]
   properties {bUsePreviewValueAsDefault=true, inputType=FunctionInput_MaterialAttributes}
  }
  MaterialExpressionFunctionInput_2 {
   class=MaterialExpressionFunctionInput
   inputs [Preview {from="", output=""}]
   properties {bUsePreviewValueAsDefault=true, inputName=MobileNormalIn, sortPriority=4, previewValue=<0 0 1 1>}
  }
  MaterialExpressionFunctionInput_3 {
   class=MaterialExpressionFunctionInput
   inputs [Preview {from="", output=""}]
   properties {bUsePreviewValueAsDefault=true, inputName=MobileFoamMaskIn, inputType=FunctionInput_Scalar, sortPriority=5}
  }
  MaterialExpressionFunctionOutput_0 {class=MaterialExpressionFunctionOutput, inputs [None {from=MaterialExpressionStaticSwitchParameter_0, output=""}]}
  MaterialExpressionFunctionOutput_1 {
   class=MaterialExpressionFunctionOutput
   inputs [None {from=MaterialExpressionStaticSwitchParameter_1, output=""}]
   properties {outputName=MobileFoamMask, sortPriority=5}
  }
  MaterialExpressionFunctionOutput_2 {
   class=MaterialExpressionFunctionOutput
   inputs [None {from=MaterialExpressionStaticSwitchParameter_2, output=""}]
   properties {outputName=MobileNormal, sortPriority=4}
  }
  MaterialExpressionFunctionOutput_3 {
   class=MaterialExpressionFunctionOutput
   inputs [None {from=MaterialExpressionNamedRerouteUsage_7, output=""}]
   properties {outputName=MobileVortexMask, sortPriority=6}
  }
  MaterialExpressionGetMaterialAttributes_0 {
   class=MaterialExpressionGetMaterialAttributes
   inputs [None {from=MaterialExpressionFunctionInput_1, output=""}]
   properties {attributeGetTypes [
     #1
     #2
     #3
     #4
    ]}
  }
  MaterialExpressionInverseLinearInterpolate_0 {
   class=MaterialExpressionInverseLinearInterpolate
   inputs [
    A {from=MaterialExpressionScalarParameter_4, output=""}
    B {from=MaterialExpressionScalarParameter_20, output=""}
    Value {from=MaterialExpressionDistance_3, output=""}
   ]
   properties {}
  }
  MaterialExpressionInverseLinearInterpolate_1 {
   class=MaterialExpressionInverseLinearInterpolate
   inputs [
    A {from=MaterialExpressionScalarParameter_4, output=""}
    B {from=MaterialExpressionScalarParameter_20, output=""}
    Value {from=MaterialExpressionDistance_0, output=""}
   ]
   properties {}
  }
  MaterialExpressionInverseLinearInterpolate_2 {
   class=MaterialExpressionInverseLinearInterpolate
   inputs [
    A {from=MaterialExpressionScalarParameter_4, output=""}
    B {from=MaterialExpressionScalarParameter_20, output=""}
    Value {from=MaterialExpressionDistance_1, output=""}
   ]
   properties {}
  }
  MaterialExpressionInverseLinearInterpolate_3 {
   class=MaterialExpressionInverseLinearInterpolate
   inputs [
    A {from=MaterialExpressionScalarParameter_4, output=""}
    B {from=MaterialExpressionScalarParameter_20, output=""}
    Value {from=MaterialExpressionDistance_2, output=""}
   ]
   properties {}
  }
  MaterialExpressionLinearInterpolate_0 {
   class=MaterialExpressionLinearInterpolate
   inputs [
    A {from=MaterialExpressionAppendVector_6}
    B {from=MaterialExpressionAppendVector_7}
    Alpha {from=MaterialExpressionNamedRerouteUsage_6}
   ]
  }
  MaterialExpressionLinearInterpolate_1 {
   class=MaterialExpressionLinearInterpolate
   inputs [
    A {from=MaterialExpressionLogarithm_0}
    B {from=MaterialExpressionLogarithm_1}
    Alpha {from=MaterialExpressionFrac_0}
   ]
  }
  MaterialExpressionLinearInterpolate_10 {
   class=MaterialExpressionLinearInterpolate
   inputs [
    A {from=MaterialExpressionScalarParameter_2}
    B {from=MaterialExpressionScalarParameter_1}
    Alpha {from=MaterialExpressionNamedRerouteUsage_2}
   ]
  }
  MaterialExpressionLinearInterpolate_11 {
   class=MaterialExpressionLinearInterpolate
   inputs [
    A {from=MaterialExpressionAppendVector_4}
    B {from=MaterialExpressionVectorParameter_4, output=RGBA}
    Alpha {from=MaterialExpressionSaturate_5}
   ]
  }
  MaterialExpressionLinearInterpolate_13 {
   class=MaterialExpressionLinearInterpolate
   inputs [
    A {from=MaterialExpressionFunctionInput_2}
    B {from=MaterialExpressionDeriveNormalZ_0}
    Alpha {from=MaterialExpressionFunctionOutput_3}
   ]
  }
  MaterialExpressionLinearInterpolate_14 {
   class=MaterialExpressionLinearInterpolate
   inputs [
    A {from=MaterialExpressionLogarithm_2}
    B {from=MaterialExpressionLogarithm_3}
    Alpha {from=MaterialExpressionFrac_5}
   ]
  }
  MaterialExpressionLinearInterpolate_15 {
   class=MaterialExpressionLinearInterpolate
   inputs [
    A {from=MaterialExpressionLogarithm_2}
    B {from=MaterialExpressionLogarithm_3}
    Alpha {from=MaterialExpressionFrac_6}
   ]
  }
  MaterialExpressionLinearInterpolate_16 {
   class=MaterialExpressionLinearInterpolate
   inputs [
    A {from=MaterialExpressionAppendVector_9}
    B {from=MaterialExpressionAppendVector_8}
    Alpha {from=MaterialExpressionAbs_4}
   ]
  }
  MaterialExpressionLinearInterpolate_17 {
   class=MaterialExpressionLinearInterpolate
   inputs [
    A {from=MaterialExpressionAppendVector_10}
    B {from=MaterialExpressionAppendVector_7}
    Alpha {from=MaterialExpressionNamedRerouteUsage_8}
   ]
  }
  MaterialExpressionLinearInterpolate_18 {
   class=MaterialExpressionLinearInterpolate
   inputs [
    A {from=""}
    B {from=MaterialExpressionDotProduct_0}
    Alpha {from=MaterialExpressionScalarParameter_16}
   ]
   properties {constA=1}
  }
  MaterialExpressionLinearInterpolate_2 {
   class=MaterialExpressionLinearInterpolate
   inputs [
    A {from=MaterialExpressionLogarithm_0}
    B {from=MaterialExpressionLogarithm_1}
    Alpha {from=MaterialExpressionFrac_1}
   ]
  }
  MaterialExpressionLinearInterpolate_3 {
   class=MaterialExpressionLinearInterpolate
   inputs [
    A {from=MaterialExpressionAppendVector_5}
    B {from=MaterialExpressionAppendVector_1}
    Alpha {from=MaterialExpressionAbs_2}
   ]
  }
  MaterialExpressionLinearInterpolate_4 {
   class=MaterialExpressionLinearInterpolate
   inputs [
    A {from=MaterialExpressionComponentMask_0}
    B {from=MaterialExpressionComponentMask_1}
    Alpha {from=MaterialExpressionPower_0}
   ]
  }
  MaterialExpressionLinearInterpolate_5 {
   class=MaterialExpressionLinearInterpolate
   inputs [
    A {from=MaterialExpressionComponentMask_0}
    B {from=MaterialExpressionComponentMask_1}
    Alpha {from=MaterialExpressionPower_1}
   ]
  }
  MaterialExpressionLinearInterpolate_6 {
   class=MaterialExpressionLinearInterpolate
   inputs [
    A {from=MaterialExpressionComponentMask_0}
    B {from=MaterialExpressionComponentMask_1}
    Alpha {from=MaterialExpressionPower_2}
   ]
  }
  MaterialExpressionLinearInterpolate_7 {
   class=MaterialExpressionLinearInterpolate
   inputs [
    A {from=MaterialExpressionComponentMask_10}
    B {from=MaterialExpressionComponentMask_9}
    Alpha {from=MaterialExpressionPower_0}
   ]
  }
  MaterialExpressionLinearInterpolate_8 {
   class=MaterialExpressionLinearInterpolate
   inputs [
    A {from=MaterialExpressionComponentMask_10}
    B {from=MaterialExpressionComponentMask_9}
    Alpha {from=MaterialExpressionPower_1}
   ]
  }
  MaterialExpressionLinearInterpolate_9 {
   class=MaterialExpressionLinearInterpolate
   inputs [
    A {from=MaterialExpressionComponentMask_10}
    B {from=MaterialExpressionComponentMask_9}
    Alpha {from=MaterialExpressionPower_2}
   ]
  }
  MaterialExpressionLogarithm_0 {class=MaterialExpressionLogarithm, inputs [None {from=MaterialExpressionScalarParameter_26, output=""}], properties {}}
  MaterialExpressionLogarithm_1 {class=MaterialExpressionLogarithm, inputs [None {from=MaterialExpressionScalarParameter_27, output=""}], properties {}}
  MaterialExpressionLogarithm_2 {class=MaterialExpressionLogarithm, inputs [None {from=MaterialExpressionScalarParameter_12, output=""}], properties {}}
  MaterialExpressionLogarithm_3 {class=MaterialExpressionLogarithm, inputs [None {from=MaterialExpressionScalarParameter_13, output=""}], properties {}}
  MaterialExpressionMaterialFunctionCall_0 {
   class=MaterialExpressionMaterialFunctionCall
   inputs [
    BaseNormal {from=MaterialExpressionGetMaterialAttributes_0, output=Normal}
    AdditionalNormal {from=MaterialExpressionDeriveNormalZ_5}
   ]
   properties {materialFunction=&BlendAngleCorrectedNormals}
  }
  MaterialExpressionMaterialFunctionCall_1 {
   class=MaterialExpressionMaterialFunctionCall
   inputs [
    "Xb1,PS4" {from=MaterialExpressionLinearInterpolate_17}
    "PC-High&&Epic,XSX,PS5" {from=MaterialExpressionLinearInterpolate_0}
    "PC-Perf,Android" {from=""}
    Switch {from=""}
    "PC-Low(NotPerf)&&Medium" {from=MaterialExpressionLinearInterpolate_17}
   ]
   properties {materialFunction=/Game/Effects/Niagara/MasterMaterials/Functions/Common/PlatformSwitch_4D}
  }
  MaterialExpressionMaterialFunctionCall_2 {
   class=MaterialExpressionMaterialFunctionCall
   inputs [
    "Xb1,PS4" {from=MaterialExpressionAppendVector_7}
    "PC-High&&Epic,XSX,PS5" {from=MaterialExpressionAppendVector_6}
    "PC-Perf,Android" {from=MaterialExpressionConstant_2}
    Switch {from=MaterialExpressionConstant_2}
    "PC-Low(NotPerf)&&Medium" {from=MaterialExpressionAppendVector_7}
   ]
   properties {materialFunction=&PlatformSwitch_4D}
  }
  MaterialExpressionMaterialFunctionCall_3 {
   class=MaterialExpressionMaterialFunctionCall
   inputs [
    "Xb1,PS4" {from=""}
    "PC-High&&Epic,XSX,PS5" {from=""}
    "PC-Perf,Android" {from=MaterialExpressionAppendVector_10}
    Switch {from=MaterialExpressionAppendVector_10}
    "PC-Low(NotPerf)&&Medium" {from=""}
   ]
   properties {materialFunction=&PlatformSwitch_4D}
  }
  MaterialExpressionMaterialFunctionCall_4 {
   class=MaterialExpressionMaterialFunctionCall
   inputs [
    UVCoords {from=MaterialExpressionNamedRerouteDeclaration_6}
    HexSize {from=MaterialExpressionScalarParameter_5}
    HexGradientSharpness {from=MaterialExpressionConstant_0}
   ]
   properties {materialFunction=/Game/Athena/Asteria/Environments/Landscape/Materials/MaterialFunctions/UtilityFunctions/HexGrid}
  }
  MaterialExpressionMax_0 {
   class=MaterialExpressionMax
   inputs [
    A {from=MaterialExpressionMultiply_35}
    B {from=MaterialExpressionFunctionInput_3}
   ]
  }
  MaterialExpressionMultiply_0 {
   class=MaterialExpressionMultiply
   inputs [
    A {from=MaterialExpressionMultiply_6}
    B {from=""}
   ]
   properties {constB=-1}
  }
  MaterialExpressionMultiply_1 {
   class=MaterialExpressionMultiply
   inputs [
    A {from=MaterialExpressionConvert_2}
    B {from=MaterialExpressionConvert_30}
   ]
  }
  MaterialExpressionMultiply_10 {
   class=MaterialExpressionMultiply
   inputs [
    A {from=MaterialExpressionLinearInterpolate_6}
    B {from=MaterialExpressionAdd_15}
   ]
  }
  MaterialExpressionMultiply_11 {
   class=MaterialExpressionMultiply
   inputs [
    A {from=MaterialExpressionNamedRerouteUsage_1}
    B {from=MaterialExpressionConvert_39}
   ]
  }
  MaterialExpressionMultiply_12 {
   class=MaterialExpressionMultiply
   inputs [
    A {from=MaterialExpressionScalarParameter_29}
    B {from=MaterialExpressionConvert_48}
   ]
  }
  MaterialExpressionMultiply_13 {
   class=MaterialExpressionMultiply
   inputs [
    A {from=MaterialExpressionVectorParameter_0, output=RGBA}
    B {from=MaterialExpressionVectorParameter_2, output=RGBA}
   ]
  }
  MaterialExpressionMultiply_14 {
   class=MaterialExpressionMultiply
   inputs [
    A {from=MaterialExpressionMultiply_17}
    B {from=MaterialExpressionScalarParameter_0}
   ]
  }
  MaterialExpressionMultiply_15 {
   class=MaterialExpressionMultiply
   inputs [
    A {from=MaterialExpressionScalarParameter_6}
    B {from=MaterialExpressionSaturate_1}
   ]
  }
  MaterialExpressionMultiply_16 {
   class=MaterialExpressionMultiply
   inputs [
    A {from=MaterialExpressionScalarParameter_3}
    B {from=MaterialExpressionMultiply_11}
   ]
  }
  MaterialExpressionMultiply_17 {
   class=MaterialExpressionMultiply
   inputs [
    A {from=MaterialExpressionNamedRerouteUsage_9}
    B {from=MaterialExpressionComponentMask_16}
   ]
  }
  MaterialExpressionMultiply_18 {
   class=MaterialExpressionMultiply
   inputs [
    A {from=MaterialExpressionConvert_0}
    B {from=""}
   ]
   properties {constB=0.5}
  }
  MaterialExpressionMultiply_19 {
   class=MaterialExpressionMultiply
   inputs [
    A {from=MaterialExpressionConvert_0}
    B {from=""}
   ]
   properties {constB=0.866}
  }
  MaterialExpressionMultiply_2 {
   class=MaterialExpressionMultiply
   inputs [
    A {from=MaterialExpressionConvert_2}
    B {from=MaterialExpressionConvert_42}
   ]
  }
  MaterialExpressionMultiply_20 {
   class=MaterialExpressionMultiply
   inputs [
    A {from=MaterialExpressionConvert_3}
    B {from=MaterialExpressionReroute_5, output=B}
   ]
  }
  MaterialExpressionMultiply_21 {
   class=MaterialExpressionMultiply
   inputs [
    A {from=MaterialExpressionConvert_4}
    B {from=""}
   ]
   properties {constB=0.5}
  }
  MaterialExpressionMultiply_22 {
   class=MaterialExpressionMultiply
   inputs [
    A {from=MaterialExpressionConvert_4}
    B {from=""}
   ]
   properties {constB=0.866}
  }
  MaterialExpressionMultiply_23 {
   class=MaterialExpressionMultiply
   inputs [
    A {from=MaterialExpressionConvert_5}
    B {from=MaterialExpressionReroute_8, output=B}
   ]
  }
  MaterialExpressionMultiply_24 {
   class=MaterialExpressionMultiply
   inputs [
    A {from=MaterialExpressionConvert_6}
    B {from=""}
   ]
   properties {constB=0.5}
  }
  MaterialExpressionMultiply_25 {
   class=MaterialExpressionMultiply
   inputs [
    A {from=MaterialExpressionConvert_6}
    B {from=""}
   ]
   properties {constB=0.866}
  }
  MaterialExpressionMultiply_26 {
   class=MaterialExpressionMultiply
   inputs [
    A {from=MaterialExpressionNormalize_1}
    B {from=MaterialExpressionConvert_8}
   ]
  }
  MaterialExpressionMultiply_27 {
   class=MaterialExpressionMultiply
   inputs [
    A {from=MaterialExpressionConvert_7}
    B {from=MaterialExpressionConvert_8}
   ]
  }
  MaterialExpressionMultiply_28 {
   class=MaterialExpressionMultiply
   inputs [
    A {from=MaterialExpressionNormalize_1}
    B {from=MaterialExpressionConvert_9}
   ]
  }
  MaterialExpressionMultiply_29 {
   class=MaterialExpressionMultiply
   inputs [
    A {from=MaterialExpressionConvert_10}
    B {from=MaterialExpressionConvert_11}
   ]
  }
  MaterialExpressionMultiply_3 {
   class=MaterialExpressionMultiply
   inputs [
    A {from=MaterialExpressionConvert_2}
    B {from=MaterialExpressionConvert_45}
   ]
  }
  MaterialExpressionMultiply_30 {
   class=MaterialExpressionMultiply
   inputs [
    A {from=MaterialExpressionNormalize_0}
    B {from=MaterialExpressionConvert_11}
   ]
  }
  MaterialExpressionMultiply_31 {
   class=MaterialExpressionMultiply
   inputs [
    A {from=MaterialExpressionNormalize_0}
    B {from=MaterialExpressionConvert_12}
   ]
  }
  MaterialExpressionMultiply_32 {
   class=MaterialExpressionMultiply
   inputs [
    A {from=MaterialExpressionConvert_13}
    B {from=MaterialExpressionConvert_14}
   ]
  }
  MaterialExpressionMultiply_33 {
   class=MaterialExpressionMultiply
   inputs [
    A {from=MaterialExpressionNormalize_2}
    B {from=MaterialExpressionConvert_14}
   ]
  }
  MaterialExpressionMultiply_34 {
   class=MaterialExpressionMultiply
   inputs [
    A {from=MaterialExpressionNormalize_2}
    B {from=MaterialExpressionConvert_15}
   ]
  }
  MaterialExpressionMultiply_35 {
   class=MaterialExpressionMultiply
   inputs [
    A {from=MaterialExpressionSaturate_8}
    B {from=""}
   ]
   properties {constB=3}
  }
  MaterialExpressionMultiply_36 {
   class=MaterialExpressionMultiply
   inputs [
    A {from=MaterialExpressionAbs_3}
    B {from=""}
   ]
   properties {constB=-1}
  }
  MaterialExpressionMultiply_37 {
   class=MaterialExpressionMultiply
   inputs [
    A {from=MaterialExpressionExponential_3}
    B {from=MaterialExpressionMultiply_55}
   ]
  }
  MaterialExpressionMultiply_38 {
   class=MaterialExpressionMultiply
   inputs [
    A {from=MaterialExpressionExponential_2}
    B {from=MaterialExpressionMultiply_55}
   ]
  }
  MaterialExpressionMultiply_39 {
   class=MaterialExpressionMultiply
   inputs [
    A {from=MaterialExpressionFrac_5}
    B {from=""}
   ]
   properties {constB=2}
  }
  MaterialExpressionMultiply_4 {
   class=MaterialExpressionMultiply
   inputs [
    A {from=MaterialExpressionAdd_18}
    B {from=""}
   ]
   properties {constB=-1}
  }
  MaterialExpressionMultiply_40 {
   class=MaterialExpressionMultiply
   inputs [
    A {from=MaterialExpressionAdd_10}
    B {from=MaterialExpressionConvert_16}
   ]
  }
  MaterialExpressionMultiply_41 {
   class=MaterialExpressionMultiply
   inputs [
    A {from=MaterialExpressionFloor_2}
    B {from=MaterialExpressionConvert_17}
   ]
  }
  MaterialExpressionMultiply_42 {
   class=MaterialExpressionMultiply
   inputs [
    A {from=MaterialExpressionMultiply_44}
    B {from=""}
   ]
   properties {constB=-1}
  }
  MaterialExpressionMultiply_43 {
   class=MaterialExpressionMultiply
   inputs [
    A {from=MaterialExpressionAdd_10}
    B {from=""}
   ]
   properties {constB=-1}
  }
  MaterialExpressionMultiply_44 {
   class=MaterialExpressionMultiply
   inputs [
    A {from=MaterialExpressionFloor_2}
    B {from=""}
   ]
   properties {constB=1.618}
  }
  MaterialExpressionMultiply_45 {
   class=MaterialExpressionMultiply
   inputs [
    A {from=MaterialExpressionFloor_3}
    B {from=""}
   ]
   properties {constB=1.618}
  }
  MaterialExpressionMultiply_46 {
   class=MaterialExpressionMultiply
   inputs [
    A {from=MaterialExpressionSubtract_7}
    B {from=MaterialExpressionScalarParameter_18}
   ]
  }
  MaterialExpressionMultiply_47 {
   class=MaterialExpressionMultiply
   inputs [
    A {from=MaterialExpressionAbs_0}
    B {from=""}
   ]
   properties {constB=-1}
  }
  MaterialExpressionMultiply_48 {
   class=MaterialExpressionMultiply
   inputs [
    A {from=MaterialExpressionNamedRerouteUsage_10}
    B {from=MaterialExpressionConvert_20}
   ]
  }
  MaterialExpressionMultiply_49 {
   class=MaterialExpressionMultiply
   inputs [
    A {from=MaterialExpressionScalarParameter_15}
    B {from=MaterialExpressionSaturate_7}
   ]
  }
  MaterialExpressionMultiply_5 {
   class=MaterialExpressionMultiply
   inputs [
    A {from=MaterialExpressionConvert_1}
    B {from=MaterialExpressionReroute_3, output=B}
   ]
  }
  MaterialExpressionMultiply_50 {
   class=MaterialExpressionMultiply
   inputs [
    A {from=MaterialExpressionComponentMask_19}
    B {from=MaterialExpressionScalarParameter_14}
   ]
  }
  MaterialExpressionMultiply_51 {
   class=MaterialExpressionMultiply
   inputs [
    A {from=MaterialExpressionAbs_1}
    B {from=""}
   ]
   properties {constB=-1}
  }
  MaterialExpressionMultiply_52 {
   class=MaterialExpressionMultiply
   inputs [
    A {from=MaterialExpressionExponential_1}
    B {from=MaterialExpressionDivide_18}
   ]
  }
  MaterialExpressionMultiply_53 {
   class=MaterialExpressionMultiply
   inputs [
    A {from=MaterialExpressionExponential_0}
    B {from=MaterialExpressionDivide_18}
   ]
  }
  MaterialExpressionMultiply_54 {
   class=MaterialExpressionMultiply
   inputs [
    A {from=MaterialExpressionFrac_0}
    B {from=""}
   ]
   properties {constB=2}
  }
  MaterialExpressionMultiply_55 {
   class=MaterialExpressionMultiply
   inputs [
    A {from=MaterialExpressionDivide_0}
    B {from=MaterialExpressionLinearInterpolate_18}
   ]
  }
  MaterialExpressionMultiply_56 {
   class=MaterialExpressionMultiply
   inputs [
    A {from=MaterialExpressionAdd_18}
    B {from=MaterialExpressionConvert_37}
   ]
  }
  MaterialExpressionMultiply_57 {
   class=MaterialExpressionMultiply
   inputs [
    A {from=MaterialExpressionFloor_0}
    B {from=MaterialExpressionConvert_38}
   ]
  }
  MaterialExpressionMultiply_6 {
   class=MaterialExpressionMultiply
   inputs [
    A {from=MaterialExpressionFloor_0}
    B {from=""}
   ]
   properties {constB=1.618}
  }
  MaterialExpressionMultiply_7 {
   class=MaterialExpressionMultiply
   inputs [
    A {from=MaterialExpressionFloor_1}
    B {from=""}
   ]
   properties {constB=1.618}
  }
  MaterialExpressionMultiply_8 {
   class=MaterialExpressionMultiply
   inputs [
    A {from=MaterialExpressionLinearInterpolate_4}
    B {from=MaterialExpressionAdd_13}
   ]
  }
  MaterialExpressionMultiply_9 {
   class=MaterialExpressionMultiply
   inputs [
    A {from=MaterialExpressionLinearInterpolate_5}
    B {from=MaterialExpressionAdd_14}
   ]
  }
  MaterialExpressionNamedRerouteDeclaration_0 {
   class=MaterialExpressionNamedRerouteDeclaration
   inputs [None {from=MaterialExpressionMaterialFunctionCall_4, output=GradientsRGBA}]
   properties {name=BlendMask, nodeColor=<rgba 0.5 0 1 1>}
  }
  MaterialExpressionNamedRerouteDeclaration_1 {
   class=MaterialExpressionNamedRerouteDeclaration
   inputs [None {from=MaterialExpressionOneMinus_2}]
   properties {name=VortexFalloff, nodeColor=<rgba 0 0.7656 1 1>}
  }
  MaterialExpressionNamedRerouteDeclaration_11 {
   class=MaterialExpressionNamedRerouteDeclaration
   inputs [None {from=MaterialExpressionPanner_0}]
   properties {name=xUV, nodeColor=<rgba 1 0 0 1>}
  }
  MaterialExpressionNamedRerouteDeclaration_15 {
   class=MaterialExpressionNamedRerouteDeclaration
   inputs [None {from=MaterialExpressionPanner_1}]
   properties {name=yUV, nodeColor=<rgba 0 1 0 1>}
  }
  MaterialExpressionNamedRerouteDeclaration_16 {
   class=MaterialExpressionNamedRerouteDeclaration
   inputs [None {from=MaterialExpressionPanner_2}]
   properties {name=zUV, nodeColor=<rgba 0 0 1 1>}
  }
  MaterialExpressionNamedRerouteDeclaration_2 {
   class=MaterialExpressionNamedRerouteDeclaration
   inputs [None {from=MaterialExpressionReroute_14, output=Pin 2}]
   properties {name=xTangents, nodeColor=<rgba 1 0.2 0.2 1>}
  }
  MaterialExpressionNamedRerouteDeclaration_3 {
   class=MaterialExpressionNamedRerouteDeclaration
   inputs [None {from=MaterialExpressionReroute_15, output=Pin 2}]
   properties {name=yTangents, nodeColor=<rgba 0.2 1 0.2 1>}
  }
  MaterialExpressionNamedRerouteDeclaration_4 {
   class=MaterialExpressionNamedRerouteDeclaration
   inputs [None {from=MaterialExpressionReroute_19, output=Pin 2}]
   properties {name=zTangents, nodeColor=<rgba 0.2 0.2 1 1>}
  }
  MaterialExpressionNamedRerouteDeclaration_6 {
   class=MaterialExpressionNamedRerouteDeclaration
   inputs [None {from=MaterialExpressionWorldPosition_0, output=XY}]
   properties {name=UVi, nodeColor=<rgba 0.9 0.009 0.009 1>}
  }
  MaterialExpressionNamedRerouteDeclaration_7 {
   class=MaterialExpressionNamedRerouteDeclaration
   inputs [None {from=MaterialExpressionSmoothStep_0}]
   properties {name=CenterMask, nodeColor=<rgba 0 0.8125 1 1>}
  }
  MaterialExpressionNamedRerouteUsage_0 {class=MaterialExpressionNamedRerouteUsage}
  MaterialExpressionNamedRerouteUsage_1 {class=MaterialExpressionNamedRerouteUsage}
  MaterialExpressionNamedRerouteUsage_10 {class=MaterialExpressionNamedRerouteUsage}
  MaterialExpressionNamedRerouteUsage_11 {class=MaterialExpressionNamedRerouteUsage}
  MaterialExpressionNamedRerouteUsage_12 {class=MaterialExpressionNamedRerouteUsage}
  MaterialExpressionNamedRerouteUsage_14 {class=MaterialExpressionNamedRerouteUsage}
  MaterialExpressionNamedRerouteUsage_16 {class=MaterialExpressionNamedRerouteUsage}
  MaterialExpressionNamedRerouteUsage_2 {class=MaterialExpressionNamedRerouteUsage}
  MaterialExpressionNamedRerouteUsage_3 {class=MaterialExpressionNamedRerouteUsage}
  MaterialExpressionNamedRerouteUsage_4 {class=MaterialExpressionNamedRerouteUsage}
  MaterialExpressionNamedRerouteUsage_5 {class=MaterialExpressionNamedRerouteUsage}
  MaterialExpressionNamedRerouteUsage_6 {class=MaterialExpressionNamedRerouteUsage}
  MaterialExpressionNamedRerouteUsage_7 {class=MaterialExpressionNamedRerouteUsage}
  MaterialExpressionNamedRerouteUsage_8 {class=MaterialExpressionNamedRerouteUsage}
  MaterialExpressionNamedRerouteUsage_9 {class=MaterialExpressionNamedRerouteUsage}
  MaterialExpressionNormalize_0 {class=MaterialExpressionNormalize, inputs [VectorInput {from=MaterialExpressionSubtract_5, output=""}]}
  MaterialExpressionNormalize_1 {class=MaterialExpressionNormalize, inputs [VectorInput {from=MaterialExpressionSubtract_4, output=""}]}
  MaterialExpressionNormalize_2 {class=MaterialExpressionNormalize, inputs [VectorInput {from=MaterialExpressionSubtract_6, output=""}]}
  MaterialExpressionOneMinus_1 {class=MaterialExpressionOneMinus, inputs [None {from=MaterialExpressionSaturate_0}]}
  MaterialExpressionOneMinus_2 {class=MaterialExpressionOneMinus, inputs [None {from=MaterialExpressionSaturate_4}]}
  MaterialExpressionOneMinus_3 {class=MaterialExpressionOneMinus, inputs [None {from=MaterialExpressionSaturate_2}]}
  MaterialExpressionOneMinus_4 {class=MaterialExpressionOneMinus, inputs [None {from=MaterialExpressionSaturate_3}]}
  MaterialExpressionPanner_0 {
   class=MaterialExpressionPanner
   inputs [
    Coordinate {from=MaterialExpressionMultiply_8}
    Time {from=""}
    Speed {from=MaterialExpressionLinearInterpolate_7}
   ]
   properties {bFractionalPart=true, speedY=1}
  }
  MaterialExpressionPanner_1 {
   class=MaterialExpressionPanner
   inputs [
    Coordinate {from=MaterialExpressionMultiply_9}
    Time {from=""}
    Speed {from=MaterialExpressionLinearInterpolate_8}
   ]
   properties {bFractionalPart=true, speedY=1}
  }
  MaterialExpressionPanner_2 {
   class=MaterialExpressionPanner
   inputs [
    Coordinate {from=MaterialExpressionMultiply_10}
    Time {from=""}
    Speed {from=MaterialExpressionLinearInterpolate_9}
   ]
   properties {bFractionalPart=true, speedY=1}
  }
  MaterialExpressionPinBase_0 {
   class=MaterialExpressionPinBase
   inputs []
   properties {reroutePins [
     Input {expression=&self:MaterialExpressionReroute_1}
     B {expression=&self:MaterialExpressionReroute_3}
    ]}
  }
  MaterialExpressionPinBase_1 {
   class=MaterialExpressionPinBase
   inputs [Output {from=MaterialExpressionMultiply_5, output=""}]
   properties {reroutePins [Output {expression=&self:MaterialExpressionReroute_4}]}
  }
  MaterialExpressionPinBase_10 {
   class=MaterialExpressionPinBase
   inputs []
   properties {reroutePins [
     Input {expression=&self:MaterialExpressionReroute_16}
     Input2 {expression=&self:MaterialExpressionReroute_17}
     B {expression=&self:MaterialExpressionReroute_18}
    ]}
  }
  MaterialExpressionPinBase_11 {
   class=MaterialExpressionPinBase
   inputs [
    Output {from=MaterialExpressionAdd_7, output=""}
    Pin 2 {from=MaterialExpressionAppendVector_3, output=""}
   ]
   properties {reroutePins [
     Output {expression=&self:MaterialExpressionReroute_23}
     Pin 2 {expression=&self:MaterialExpressionReroute_19}
    ]}
  }
  MaterialExpressionPinBase_2 {
   class=MaterialExpressionPinBase
   inputs []
   properties {reroutePins [
     Input {expression=&self:MaterialExpressionReroute_2}
     B {expression=&self:MaterialExpressionReroute_5}
    ]}
  }
  MaterialExpressionPinBase_3 {
   class=MaterialExpressionPinBase
   inputs [Output {from=MaterialExpressionMultiply_20, output=""}]
   properties {reroutePins [Output {expression=&self:MaterialExpressionReroute_6}]}
  }
  MaterialExpressionPinBase_4 {
   class=MaterialExpressionPinBase
   inputs []
   properties {reroutePins [
     Input {expression=&self:MaterialExpressionReroute_7}
     B {expression=&self:MaterialExpressionReroute_8}
    ]}
  }
  MaterialExpressionPinBase_5 {
   class=MaterialExpressionPinBase
   inputs [Output {from=MaterialExpressionMultiply_23, output=""}]
   properties {reroutePins [Output {expression=&self:MaterialExpressionReroute_9}]}
  }
  MaterialExpressionPinBase_6 {
   class=MaterialExpressionPinBase
   inputs []
   properties {reroutePins [
     Input {expression=&self:MaterialExpressionReroute_10}
     Input2 {expression=&self:MaterialExpressionReroute_11}
     B {expression=&self:MaterialExpressionReroute_12}
    ]}
  }
  MaterialExpressionPinBase_7 {
   class=MaterialExpressionPinBase
   inputs [
    Output {from=MaterialExpressionAdd_4, output=""}
    Pin 2 {from=MaterialExpressionAppendVector_0, output=""}
   ]
   properties {reroutePins [
     Output {expression=&self:MaterialExpressionReroute_13}
     Pin 2 {expression=&self:MaterialExpressionReroute_14}
    ]}
  }
  MaterialExpressionPinBase_8 {
   class=MaterialExpressionPinBase
   inputs []
   properties {reroutePins [
     Input {expression=&self:MaterialExpressionReroute_24}
     Input2 {expression=&self:MaterialExpressionReroute_25}
     B {expression=&self:MaterialExpressionReroute_26}
    ]}
  }
  MaterialExpressionPinBase_9 {
   class=MaterialExpressionPinBase
   inputs [
    Output {from=MaterialExpressionAdd_6, output=""}
    Pin 2 {from=MaterialExpressionAppendVector_2, output=""}
   ]
   properties {reroutePins [
     Output {expression=&self:MaterialExpressionReroute_27}
     Pin 2 {expression=&self:MaterialExpressionReroute_15}
    ]}
  }
  MaterialExpressionPower_0 {
   class=MaterialExpressionPower
   inputs [
    Base {from=MaterialExpressionOneMinus_1}
    Exp {from=MaterialExpressionScalarParameter_21}
   ]
  }
  MaterialExpressionPower_1 {
   class=MaterialExpressionPower
   inputs [
    Base {from=MaterialExpressionOneMinus_3}
    Exp {from=MaterialExpressionScalarParameter_21}
   ]
  }
  MaterialExpressionPower_2 {
   class=MaterialExpressionPower
   inputs [
    Base {from=MaterialExpressionOneMinus_4}
    Exp {from=MaterialExpressionScalarParameter_21}
   ]
  }
  MaterialExpressionPower_3 {
   class=MaterialExpressionPower
   inputs [
    Base {from=MaterialExpressionComponentMask_15}
    Exp {from=""}
   ]
  }
  MaterialExpressionPower_4 {
   class=MaterialExpressionPower
   inputs [
    Base {from=MaterialExpressionComponentMask_20}
    Exp {from=""}
   ]
  }
  MaterialExpressionReroute_0 {class=MaterialExpressionReroute, inputs [None {from=MaterialExpressionScalarParameter_5}]}
  MaterialExpressionReroute_1 {class=MaterialExpressionReroute, inputs [None {from=MaterialExpressionMaterialFunctionCall_4, output=Grid1}]}
  MaterialExpressionReroute_10 {class=MaterialExpressionReroute, inputs [None {from=MaterialExpressionCollectionParameter_1}]}
  MaterialExpressionReroute_11 {class=MaterialExpressionReroute, inputs [None {from=MaterialExpressionSubtract_0}]}
  MaterialExpressionReroute_12 {class=MaterialExpressionReroute, inputs [None {from=MaterialExpressionReroute_4, output=Output}]}
  MaterialExpressionReroute_13 {class=MaterialExpressionReroute, inputs [None {from=MaterialExpressionAdd_4}]}
  MaterialExpressionReroute_14 {class=MaterialExpressionReroute, inputs [None {from=MaterialExpressionAppendVector_0}]}
  MaterialExpressionReroute_15 {class=MaterialExpressionReroute, inputs [None {from=MaterialExpressionAppendVector_2}]}
  MaterialExpressionReroute_16 {class=MaterialExpressionReroute, inputs [None {from=MaterialExpressionCollectionParameter_1}]}
  MaterialExpressionReroute_17 {class=MaterialExpressionReroute, inputs [None {from=MaterialExpressionSubtract_3}]}
  MaterialExpressionReroute_18 {class=MaterialExpressionReroute, inputs [None {from=MaterialExpressionReroute_9, output=Output}]}
  MaterialExpressionReroute_19 {class=MaterialExpressionReroute, inputs [None {from=MaterialExpressionAppendVector_3}]}
  MaterialExpressionReroute_2 {class=MaterialExpressionReroute, inputs [None {from=MaterialExpressionMaterialFunctionCall_4, output=Grid2}]}
  MaterialExpressionReroute_20 {class=MaterialExpressionReroute, inputs [None {from=MaterialExpressionFunctionInput_1}]}
  MaterialExpressionReroute_23 {class=MaterialExpressionReroute, inputs [None {from=MaterialExpressionAdd_7}]}
  MaterialExpressionReroute_24 {class=MaterialExpressionReroute, inputs [None {from=MaterialExpressionCollectionParameter_1}]}
  MaterialExpressionReroute_25 {class=MaterialExpressionReroute, inputs [None {from=MaterialExpressionSubtract_2}]}
  MaterialExpressionReroute_26 {class=MaterialExpressionReroute, inputs [None {from=MaterialExpressionReroute_6, output=Output}]}
  MaterialExpressionReroute_27 {class=MaterialExpressionReroute, inputs [None {from=MaterialExpressionAdd_6}]}
  MaterialExpressionReroute_3 {class=MaterialExpressionReroute, inputs [None {from=MaterialExpressionReroute_0}]}
  MaterialExpressionReroute_4 {class=MaterialExpressionReroute, inputs [None {from=MaterialExpressionMultiply_5}]}
  MaterialExpressionReroute_5 {class=MaterialExpressionReroute, inputs [None {from=MaterialExpressionReroute_0}]}
  MaterialExpressionReroute_6 {class=MaterialExpressionReroute, inputs [None {from=MaterialExpressionMultiply_20}]}
  MaterialExpressionReroute_7 {class=MaterialExpressionReroute, inputs [None {from=MaterialExpressionMaterialFunctionCall_4, output=Grid3}]}
  MaterialExpressionReroute_8 {class=MaterialExpressionReroute, inputs [None {from=MaterialExpressionReroute_0}]}
  MaterialExpressionReroute_9 {class=MaterialExpressionReroute, inputs [None {from=MaterialExpressionMultiply_23}]}
  MaterialExpressionRotator_0 {
   class=MaterialExpressionRotator
   inputs [
    Coordinate {from=MaterialExpressionConstantBiasScale_10, output=""}
    Time {from=MaterialExpressionMultiply_0, output=""}
   ]
   properties {centerX=0, centerY=0, speed=1}
  }
  MaterialExpressionRotator_1 {
   class=MaterialExpressionRotator
   inputs [
    Coordinate {from=MaterialExpressionConstantBiasScale_11, output=""}
    Time {from=MaterialExpressionMultiply_4, output=""}
   ]
   properties {centerX=0, centerY=0, speed=1}
  }
  MaterialExpressionRotator_2 {
   class=MaterialExpressionRotator
   inputs [
    Coordinate {from=MaterialExpressionMultiply_38, output=""}
    Time {from=MaterialExpressionAdd_10, output=""}
   ]
   properties {speed=1}
  }
  MaterialExpressionRotator_3 {
   class=MaterialExpressionRotator
   inputs [
    Coordinate {from=MaterialExpressionMultiply_53, output=""}
    Time {from=MaterialExpressionAdd_18, output=""}
   ]
   properties {speed=1}
  }
  MaterialExpressionRotator_4 {
   class=MaterialExpressionRotator
   inputs [
    Coordinate {from=MaterialExpressionMultiply_52, output=""}
    Time {from=MaterialExpressionMultiply_6, output=""}
   ]
   properties {speed=1}
  }
  MaterialExpressionRotator_5 {
   class=MaterialExpressionRotator
   inputs [
    Coordinate {from=MaterialExpressionMultiply_37, output=""}
    Time {from=MaterialExpressionMultiply_44, output=""}
   ]
   properties {speed=1}
  }
  MaterialExpressionRotator_6 {
   class=MaterialExpressionRotator
   inputs [
    Coordinate {from=MaterialExpressionConstantBiasScale_0, output=""}
    Time {from=MaterialExpressionMultiply_42, output=""}
   ]
   properties {centerX=0, centerY=0, speed=1}
  }
  MaterialExpressionRotator_7 {
   class=MaterialExpressionRotator
   inputs [
    Coordinate {from=MaterialExpressionConstantBiasScale_1, output=""}
    Time {from=MaterialExpressionMultiply_43, output=""}
   ]
   properties {centerX=0, centerY=0, speed=1}
  }
  MaterialExpressionSaturate_0 {class=MaterialExpressionSaturate, inputs [None {from=MaterialExpressionInverseLinearInterpolate_1}]}
  MaterialExpressionSaturate_1 {class=MaterialExpressionSaturate, inputs [None {from=MaterialExpressionSubtract_1}]}
  MaterialExpressionSaturate_2 {class=MaterialExpressionSaturate, inputs [None {from=MaterialExpressionInverseLinearInterpolate_2}]}
  MaterialExpressionSaturate_3 {class=MaterialExpressionSaturate, inputs [None {from=MaterialExpressionInverseLinearInterpolate_3}]}
  MaterialExpressionSaturate_4 {class=MaterialExpressionSaturate, inputs [None {from=MaterialExpressionInverseLinearInterpolate_0}]}
  MaterialExpressionSaturate_5 {class=MaterialExpressionSaturate, inputs [None {from=MaterialExpressionMultiply_15}]}
  MaterialExpressionSaturate_6 {class=MaterialExpressionSaturate, inputs [None {from=MaterialExpressionLinearInterpolate_10}]}
  MaterialExpressionSaturate_7 {class=MaterialExpressionSaturate, inputs [None {from=MaterialExpressionSubtract_12}]}
  MaterialExpressionSaturate_8 {class=MaterialExpressionSaturate, inputs [None {from=MaterialExpressionMultiply_49}]}
  MaterialExpressionScalarParameter_0 {class=MaterialExpressionScalarParameter, properties {defaultValue=100, group=Vortex, parameterName=Vortex WPO Height Range}}
  MaterialExpressionScalarParameter_1 {class=MaterialExpressionScalarParameter, properties {defaultValue=-0.25, group=Vortex, parameterName=Vortex Foam Reduction Near}}
  MaterialExpressionScalarParameter_10 {class=MaterialExpressionScalarParameter, properties {defaultValue=0.67, group=Vortex, parameterName=Mobile Vortex Flip Duration}}
  MaterialExpressionScalarParameter_11 {class=MaterialExpressionScalarParameter, properties {defaultValue=3300, group=Vortex, parameterName=Vortex UV Size}}
  MaterialExpressionScalarParameter_12 {class=MaterialExpressionScalarParameter, properties {defaultValue=1, group=Vortex, parameterName=Mobile Vortex UV Scale Start}}
  MaterialExpressionScalarParameter_13 {class=MaterialExpressionScalarParameter, properties {defaultValue=2, group=Vortex, parameterName=Mobile Vortex UV Scale End}}
  MaterialExpressionScalarParameter_14 {class=MaterialExpressionScalarParameter, properties {defaultValue=0.1, group=Vortex, parameterName=Mobile Vortex Normal Strength}}
  MaterialExpressionScalarParameter_15 {class=MaterialExpressionScalarParameter, properties {defaultValue=0.8, group=Vortex, parameterName=Vortex Foam Intensity}}
  MaterialExpressionScalarParameter_16 {class=MaterialExpressionScalarParameter, properties {defaultValue=0.5, group=Vortex, parameterName=Mobile Vortex UV Pinch}}
  MaterialExpressionScalarParameter_17 {class=MaterialExpressionScalarParameter, properties {defaultValue=13584.7061, group=Vortex, parameterName=VortexFlipNoiseSize}}
  MaterialExpressionScalarParameter_18 {class=MaterialExpressionScalarParameter, properties {defaultValue=1, group=Vortex, parameterName=VortexFlipNoiseIntensity}}
  MaterialExpressionScalarParameter_2 {class=MaterialExpressionScalarParameter, properties {defaultValue=1, group=Vortex, parameterName=Vortex Foam Reduction Far}}
  MaterialExpressionScalarParameter_20 {class=MaterialExpressionScalarParameter, properties {defaultValue=5190.6958, group=Vortex, parameterName=Vortex Range}}
  MaterialExpressionScalarParameter_21 {class=MaterialExpressionScalarParameter, properties {defaultValue=3, group=Vortex, parameterName=Vortex Speed Falloff Pow}}
  MaterialExpressionScalarParameter_22 {class=MaterialExpressionScalarParameter, properties {defaultValue=3300, group=Vortex, parameterName=Vortex Center Base UV Size}}
  MaterialExpressionScalarParameter_23 {class=MaterialExpressionScalarParameter, properties {defaultValue=1300, group=Vortex, parameterName=Vortex Center Outer Range}}
  MaterialExpressionScalarParameter_24 {class=MaterialExpressionScalarParameter, properties {defaultValue=900, group=Vortex, parameterName=Vortex Center Inner Range}}
  MaterialExpressionScalarParameter_25 {class=MaterialExpressionScalarParameter, properties {defaultValue=0.67, group=Vortex, parameterName=Vortex Center Flip Duration}}
  MaterialExpressionScalarParameter_26 {class=MaterialExpressionScalarParameter, properties {defaultValue=1.5, group=Vortex, parameterName=Vortex Center UV Scale Start}}
  MaterialExpressionScalarParameter_27 {class=MaterialExpressionScalarParameter, properties {defaultValue=0.5, group=Vortex, parameterName=Vortex Center UV Scale End}}
  MaterialExpressionScalarParameter_29 {class=MaterialExpressionScalarParameter, properties {defaultValue=0.5, group=Vortex, parameterName=Vortex Center Normal Strength}}
  MaterialExpressionScalarParameter_3 {class=MaterialExpressionScalarParameter, properties {defaultValue=0.5, group=Vortex, parameterName=Vortex Normal Strength}}
  MaterialExpressionScalarParameter_4 {class=MaterialExpressionScalarParameter, properties {defaultValue=900, group=Vortex, parameterName=Vortex Range Interior}}
  MaterialExpressionScalarParameter_5 {class=MaterialExpressionScalarParameter, properties {defaultValue=800, group=Vortex, parameterName=Vortex Tile Size}}
  MaterialExpressionScalarParameter_6 {class=MaterialExpressionScalarParameter, properties {defaultValue=0.8, group=Vortex, parameterName=Vortex Foam Intensity}}
  MaterialExpressionScalarParameter_7 {class=MaterialExpressionScalarParameter, properties {defaultValue=4500, group=Vortex, parameterName=Mobile Vortex Base UV Size}}
  MaterialExpressionSetMaterialAttributes_0 {
   class=MaterialExpressionSetMaterialAttributes
   inputs [
    MaterialAttributes {from=MaterialExpressionGetMaterialAttributes_0, output=MaterialAttributes}
    BaseColor {from=MaterialExpressionComponentMask_13, output=""}
    Opacity {from=MaterialExpressionComponentMask_17, output=""}
    Normal {from=MaterialExpressionMaterialFunctionCall_0, output=Result}
    WorldPositionOffset {from=MaterialExpressionAdd_8, output=""}
   ]
   properties {attributeSetTypes [
     #1
     #2
     #3
     #4
    ]}
  }
  MaterialExpressionSmoothStep_0 {
   class=MaterialExpressionSmoothStep
   inputs [
    Min {from=MaterialExpressionScalarParameter_23, output=""}
    Max {from=MaterialExpressionScalarParameter_24, output=""}
    Value {from=MaterialExpressionDistance_4, output=""}
   ]
   properties {}
  }
  MaterialExpressionStaticSwitchParameter_0 {
   class=MaterialExpressionStaticSwitchParameter
   inputs [
    True {from=MaterialExpressionSetMaterialAttributes_0, output=""}
    False {from=MaterialExpressionReroute_20, output=""}
   ]
   properties {group=Vortex, parameterName=Use Zero Point Vortex, sortPriority=-1}
  }
  MaterialExpressionStaticSwitchParameter_1 {
   class=MaterialExpressionStaticSwitchParameter
   inputs [
    True {from=MaterialExpressionMax_0, output=""}
    False {from=MaterialExpressionFunctionInput_3, output=""}
   ]
   properties {group=Vortex, parameterName=Use Zero Point Vortex, sortPriority=-1}
  }
  MaterialExpressionStaticSwitchParameter_2 {
   class=MaterialExpressionStaticSwitchParameter
   inputs [
    True {from=MaterialExpressionLinearInterpolate_13, output=""}
    False {from=MaterialExpressionFunctionInput_2, output=""}
   ]
   properties {group=Vortex, parameterName=Use Zero Point Vortex, sortPriority=-1}
  }
  MaterialExpressionSubtract_0 {
   class=MaterialExpressionSubtract
   inputs [
    A {from=MaterialExpressionNamedRerouteUsage_11}
    B {from=MaterialExpressionReroute_4, output=Output}
   ]
  }
  MaterialExpressionSubtract_1 {
   class=MaterialExpressionSubtract
   inputs [
    A {from=MaterialExpressionPower_3}
    B {from=MaterialExpressionSaturate_6}
   ]
  }
  MaterialExpressionSubtract_10 {
   class=MaterialExpressionSubtract
   inputs [
    A {from=MaterialExpressionMultiply_54}
    B {from=""}
   ]
  }
  MaterialExpressionSubtract_11 {
   class=MaterialExpressionSubtract
   inputs [
    A {from=MaterialExpressionMultiply_39}
    B {from=""}
   ]
  }
  MaterialExpressionSubtract_12 {
   class=MaterialExpressionSubtract
   inputs [
    A {from=MaterialExpressionPower_4}
    B {from=MaterialExpressionSaturate_6}
   ]
  }
  MaterialExpressionSubtract_2 {
   class=MaterialExpressionSubtract
   inputs [
    A {from=MaterialExpressionNamedRerouteUsage_11}
    B {from=MaterialExpressionReroute_6, output=Output}
   ]
  }
  MaterialExpressionSubtract_3 {
   class=MaterialExpressionSubtract
   inputs [
    A {from=MaterialExpressionNamedRerouteUsage_11}
    B {from=MaterialExpressionReroute_9, output=Output}
   ]
  }
  MaterialExpressionSubtract_4 {
   class=MaterialExpressionSubtract
   inputs [
    A {from=MaterialExpressionComponentMask_2}
    B {from=MaterialExpressionReroute_12, output=B}
   ]
  }
  MaterialExpressionSubtract_5 {
   class=MaterialExpressionSubtract
   inputs [
    A {from=MaterialExpressionComponentMask_3}
    B {from=MaterialExpressionReroute_26, output=B}
   ]
  }
  MaterialExpressionSubtract_6 {
   class=MaterialExpressionSubtract
   inputs [
    A {from=MaterialExpressionComponentMask_4}
    B {from=MaterialExpressionReroute_18, output=B}
   ]
  }
  MaterialExpressionSubtract_7 {
   class=MaterialExpressionSubtract
   inputs [
    A {from=MaterialExpressionComponentMask_8}
    B {from=""}
   ]
   properties {constB=0.5}
  }
  MaterialExpressionSubtract_8 {
   class=MaterialExpressionSubtract
   inputs [
    A {from=MaterialExpressionWorldPosition_1, output=XY}
    B {from=MaterialExpressionComponentMask_18}
   ]
  }
  MaterialExpressionSubtract_9 {
   class=MaterialExpressionSubtract
   inputs [
    A {from=MaterialExpressionWorldPosition_4, output=XY}
    B {from=MaterialExpressionComponentMask_12}
   ]
  }
  MaterialExpressionTextureObject_0 {class=MaterialExpressionTextureObject, inputs [], properties {texture=/Game/Global/Textures/Default/Tileable_Patterns/T_RiverFoam_new_M}}
  MaterialExpressionTextureObject_1 {class=MaterialExpressionTextureObject, inputs [], properties {texture=&T_RiverFoam_new_M}}
  MaterialExpressionTextureObject_2 {class=MaterialExpressionTextureObject, inputs [], properties {texture=&T_RiverFoam_new_M}}
  MaterialExpressionTextureSample_0 {
   class=MaterialExpressionTextureSample
   inputs [
    UVs {from=MaterialExpressionNamedRerouteUsage_12}
    Tex {from=MaterialExpressionTextureObject_0}
    Apply View MipBias {from=""}
   ]
   properties {samplerType=SAMPLERTYPE_Normal, texture=/FortWater/Textures/T_Water_N}
  }
  MaterialExpressionTextureSample_1 {
   class=MaterialExpressionTextureSample
   inputs [
    UVs {from=MaterialExpressionNamedRerouteUsage_14}
    Tex {from=MaterialExpressionTextureObject_0}
    Apply View MipBias {from=""}
   ]
   properties {samplerType=SAMPLERTYPE_Normal, texture=&T_Water_N}
  }
  MaterialExpressionTextureSample_2 {
   class=MaterialExpressionTextureSample
   inputs [
    UVs {from=MaterialExpressionNamedRerouteUsage_16}
    Tex {from=MaterialExpressionTextureObject_0}
    Apply View MipBias {from=""}
   ]
   properties {samplerType=SAMPLERTYPE_Normal, texture=&T_Water_N}
  }
  MaterialExpressionTextureSample_3 {
   class=MaterialExpressionTextureSample
   inputs [
    UVs {from=MaterialExpressionDivide_12}
    Tex {from=""}
    Apply View MipBias {from=""}
   ]
   properties {samplerType=SAMPLERTYPE_LinearColor, texture=&HighContrastNoise}
  }
  MaterialExpressionTextureSample_4 {
   class=MaterialExpressionTextureSample
   inputs [
    UVs {from=MaterialExpressionAdd_20}
    Tex {from=MaterialExpressionTextureObject_1}
    Apply View MipBias {from=""}
   ]
   properties {samplerType=SAMPLERTYPE_Normal, texture=&T_Water_N}
  }
  MaterialExpressionTextureSample_5 {
   class=MaterialExpressionTextureSample
   inputs [
    UVs {from=MaterialExpressionAdd_19}
    Tex {from=MaterialExpressionTextureObject_1}
    Apply View MipBias {from=""}
   ]
   properties {samplerType=SAMPLERTYPE_Normal, texture=&T_Water_N}
  }
  MaterialExpressionTextureSample_6 {
   class=MaterialExpressionTextureSample
   inputs [
    UVs {from=MaterialExpressionAdd_12}
    Tex {from=MaterialExpressionTextureObject_2}
    Apply View MipBias {from=""}
   ]
   properties {samplerType=SAMPLERTYPE_Normal, texture=&T_Water_N}
  }
  MaterialExpressionTextureSample_7 {
   class=MaterialExpressionTextureSample
   inputs [
    UVs {from=MaterialExpressionAdd_11}
    Tex {from=MaterialExpressionTextureObject_2}
    Apply View MipBias {from=""}
   ]
   properties {samplerType=SAMPLERTYPE_Normal, texture=&T_Water_N}
  }
  MaterialExpressionTime_0 {class=MaterialExpressionTime}
  MaterialExpressionTime_3 {class=MaterialExpressionTime}
  MaterialExpressionVectorNoise_0 {class=MaterialExpressionVectorNoise, inputs [World Position {from=MaterialExpressionConvert_25, output=""}], properties {}}
  MaterialExpressionVectorNoise_1 {class=MaterialExpressionVectorNoise, inputs [World Position {from=MaterialExpressionConvert_26, output=""}], properties {}}
  MaterialExpressionVectorNoise_2 {class=MaterialExpressionVectorNoise, inputs [World Position {from=MaterialExpressionConvert_27, output=""}], properties {}}
  MaterialExpressionVectorParameter_0 {
   class=MaterialExpressionVectorParameter
   properties {group=Vortex, parameterName=Vortex UV Scale Near Far, defaultValue=<rgba 0.2 0.5 1 1.5>}
  }
  MaterialExpressionVectorParameter_1 {class=MaterialExpressionVectorParameter, properties {defaultValue=<rgba -11008 -1280 0 0>}}
  MaterialExpressionVectorParameter_2 {class=MaterialExpressionVectorParameter, properties {group=Vortex, parameterName=Vortex Pan Speed Near Far, defaultValue=<rgba 0 1 0 0.25>}}
  MaterialExpressionVectorParameter_3 {class=MaterialExpressionVectorParameter, properties {parameterName=VortexCenter, defaultValue=<rgba -11008 -1280 4540.2563 0>}}
  MaterialExpressionVectorParameter_4 {class=MaterialExpressionVectorParameter, properties {group=Vortex, parameterName=Vortex Foam Color, defaultValue=<rgba 0.6823 0.9535 1 1>}}
  MaterialExpressionVectorParameter_5 {class=MaterialExpressionVectorParameter, properties {parameterName=VortexCenter, defaultValue=<rgba -11008 -1280 4540.2563 0>}}
  MaterialExpressionWorldPosition_0 {class=MaterialExpressionWorldPosition, inputs [], properties {}}
  MaterialExpressionWorldPosition_1 {class=MaterialExpressionWorldPosition, inputs [], properties {}}
  MaterialExpressionWorldPosition_2 {class=MaterialExpressionWorldPosition, inputs [], properties {}}
  MaterialExpressionWorldPosition_3 {class=MaterialExpressionWorldPosition, inputs [], properties {}}
  MaterialExpressionWorldPosition_4 {class=MaterialExpressionWorldPosition, inputs [], properties {}}
 }}
