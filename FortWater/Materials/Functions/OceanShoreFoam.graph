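# Material function that layers shoreline and wave foam onto ocean water attributes, also outputting mask, color, normal and scattering.
# Exponential falloff of wave displacement and depth combines with world-space foam textures, minus simulation, far-mask and Epic-only physics-velocity erasure.
# Low, Medium, Forward and Mobile paths use a cheaper depth-edge mask, blending only color and opacity.
# Landscape-derived flowmap and bobbing branches exist but reach no output.
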MaterialFunction /FortWater/Materials/Functions/OceanShoreFoam
graph {nodes [
  n1 = TextureObjectParameter() {
   parameterName=FoamTexture
   samplerType=SAMPLERTYPE_LinearGrayscale
   texture=/Game/Global/Textures/Default/Caustics/03_WaterFoam_AccentSwirls_H
  }
  n2 = MaterialFunctionCall() {materialFunction=/Water/Materials/Functions/GetWaterTime}
  n3 = ScalarParameter() {defaultValue=0.5, parameterName=Foam Speed}
  n4 = ComponentMask(None=WorldPosition() {worldPositionShaderOffset=WPT_ExcludeAllShaderOffsets}.XYZ) {g=true, r=true}
  n5 = ScalarParameter() {defaultValue=1250, parameterName=FoamScale}
  n6 = CollectionParameter() {collection=/Landmass/Landscape/BlueprintBrushes/MPC/MPC_Landscape, parameterName=RTResX}
  n7 = Subtract(
   A=ComponentMask(None=WorldPosition().XYZ) {g=true, r=true}
   B=ComponentMask(None=CollectionParameter() {collection=&MPC_Landscape, parameterName=LandscapeLocation}) {g=true, r=true}
  )
  n8 = ComponentMask(None=CollectionParameter() {collection=&MPC_Landscape, parameterName=RTWorldSize}) {g=true, r=true}
  n9 = MaterialFunctionCall!(
   Height Map=TextureObjectParameter() {parameterName=WaterVelocityAndHeight, texture=&Black}
   Normal Map Intensity=ScalarParameter() {defaultValue=-8, parameterName=DerivedShoreNormalIntensity}
   Height Map UV Offset=Multiply(A=Divide(B=@n6), B=ScalarParameter() {defaultValue=8, parameterName=DerivedNormalUVOffset})
   Coordinates=Saturate(None=Divide(A=@n7, B=@n8))
   Height Map Channel Selector=Constant4Vector() {constant=<rgba 0 0 0 1>}.RGBA
  ) {materialFunction=/Engine/Functions/Engine_MaterialFunctions03/Procedurals/NormalFromHeightmap}
  n10 = Multiply(A=Normalize(VectorInput=ComponentMask(None=@n9.Result) {g=true, r=true}), B=ScalarParameter() {parameterName=Gradient Flow})
  ComponentMask(None=MaterialFunctionCall!(
    Diffuse=@n1
    Normal Map
    "Flow Vector Map (see tooltip)"=@n10
    UV texture 2 offset and Scale=VectorParameter() {parameterName=Flowmap UV2 Offset and Scale, defaultValue=<rgba 0.618 1.618 1.382 0>}.RGB
    "Time (see tooltip)"=MaterialFunctionCall(Time=@n2.WaterTime, Speed=@n3) {materialFunction=&TimeWithSpeedVariable}.Time
    UVs=Divide(A=@n4, B=@n5)
   ) {materialFunction=&/WaterFlowMap}.Diffuse) {r=true}
  n11 = GetMaterialAttributes(None=FunctionInput() {bUsePreviewValueAsDefault=true, inputName=Attributes, inputType=FunctionInput_MaterialAttributes}) {attributeGetTypes [#1]}
  ComponentMask(None=@n11.WorldPositionOffset) {b=true, desc=WPO}
  n12 = FunctionInput() {bUsePreviewValueAsDefault=true, inputName=Mask, inputType=FunctionInput_Scalar, sortPriority=2, previewValue=<1 0 0 1>}
  n13 = ComponentMask(None=WorldPosition().XYZ) {b=true}
  n14 = ComponentMask(None=WorldPosition() {worldPositionShaderOffset=WPT_ExcludeAllShaderOffsets}.XYZ) {b=true}
  n15 = ScalarParameter() {defaultValue=-0.03, group=Foam and Pondscum, parameterName="WPO Foam "}
  n16 = Saturate(None=Divide(A=PixelDepth().R) {constB=15000})
  n17 = Multiply(A=FunctionInput() {bUsePreviewValueAsDefault=true, inputName=Depth, inputType=FunctionInput_Scalar, sortPriority=3}) {desc=Depth}
  MaterialExpressionAdd_6 = Add(
   A=Multiply(A=Subtract(A=@n13, B=@n14), B=LinearInterpolate(A=@n15, Alpha=@n16) {constB=0})
   B=Multiply(A=@n17, B=ScalarParameter() {defaultValue=0.004, group=Foam and Pondscum, parameterName=depth foam amount})
  )
  n18 = Custom!(x=@MaterialExpressionAdd_6) {
   code="return exp(-x);"
   outputType=CMOT_Float1
   inputs [{inputName=x, input {expression=&self:MaterialExpressionAdd_6}}]
  }
  n19 = Subtract(A=@n18, B=ScalarParameter() {group=Foam and Pondscum, parameterName=Foam Dark Color Cutoff})
  n20 = Add(B=ComponentMask(None=WorldPosition() {worldPositionShaderOffset=WPT_ExcludeAllShaderOffsets}.XYZ) {g=true, r=true})
  n21 = TextureSample!(
   UVs=Divide(A=@n20, B=ScalarParameter() {defaultValue=2175, group=Foam and Pondscum, parameterName=FoamScale2})
   Tex
   Bias
   Apply View MipBias
  ) {mipValueMode=TMVM_MipBias, samplerSource=SSM_Wrap_WorldGroupSettings, texture=/FortWater/Textures/T_SwirlFlow_02_D}
  n22 = Multiply(A=@n21.R, B=ScalarParameter() {defaultValue=1, group=Foam and Pondscum, parameterName=FoamScatter Depth TEx})
  n23 = StaticSwitchParameter(True=Saturate(None=Multiply(A=@n19, B=@n22)), False=Constant()) {defaultValue=true, parameterName=Use Ocean Foam}
  n24 = VectorParameter() {group=Water Shading Model, parameterName=FoamScattering, defaultValue=<rgba 0 0.1 0.2 5>}
  n25 = Divide(A=@n24.A) {constB=100}
  FunctionOutput(None=Multiply(A=Multiply(A=@n12, B=@n23), B=Multiply(A=@n24.RGB, B=@n25))) {outputName=Scatter Mask, sortPriority=3}
  n26 = GetMaterialAttributes(None=@n11.MaterialAttributes) {attributeGetTypes [
    #2
    #3
    #4
    #5
    #6
   ]}
  n27 = VectorParameter() {group=Foam and Pondscum, parameterName=Foam Color 1, defaultValue=<rgba 0.2 0.7 1 0.4>}
  n28 = VectorParameter() {group=Foam and Pondscum, parameterName=Foam Color 2, defaultValue=<rgba 1 1 1 0.38>}
  n29 = Multiply(A=@n28.RGB, B=@n28.RGB)
  n30 = Subtract(A=@n18) {constB=0.5}
  n31 = ScalarParameter() {defaultValue=0.5, group=Foam and Pondscum, parameterName=Foam Foam Erode Min}
  n32 = ScalarParameter() {defaultValue=6, group=Foam and Pondscum, parameterName=Foam Texture 2 Height}
  n33 = Divide(A=@n20, B=ScalarParameter() {defaultValue=1250, group=Foam and Pondscum, parameterName=FoamScale})
  n34 = TextureSample(UVs=@n33) {
   mipValueMode=TMVM_MipBias
   samplerSource=SSM_Wrap_WorldGroupSettings
   samplerType=SAMPLERTYPE_LinearColor
   texture=/FortWater/Textures/T_foam_Normal2
  }
  n35 = Subtract(A=LinearInterpolate(A=Multiply(A=@n32, B=@n21.R), B=@n34.A, Alpha=ScalarParameter() {defaultValue=0.9, group=Foam and Pondscum, parameterName=FoamAlphaContrast})) {constB=0.5}
  n36 = ScalarParameter() {defaultValue=0.75, group=Foam and Pondscum, parameterName=Foam TexHeight}
  n37 = FunctionInput() {bUsePreviewValueAsDefault=true, inputName=Sim Erasure, inputType=FunctionInput_Scalar, sortPriority=1}
  n38 = SamplePhysicsVectorField!(World Position=WorldPosition().XYZ) {fieldTarget=Vector_LinearVelocity}
  n39 = QualitySwitch(
   Default=Constant()
   Epic=Multiply(
    A=MaterialFunctionCall(Vector 3=@n38) {materialFunction=/Engine/Functions/Engine_MaterialFunctions02/Utility/VectorLength}.V3 Length
    B=ScalarParameter() {defaultValue=2, group=Foam and Pondscum, parameterName=Water Physics Force Foam Erasure}
   )
  )
  n40 = FunctionInput() {bUsePreviewValueAsDefault=true, inputName=Far Mask, inputType=FunctionInput_Scalar, sortPriority=2}
  n41 = ScalarParameter() {defaultValue=0.15, parameterName=Far Foam Remove}
  n42 = Subtract(A=Add(A=Min(A=@n30, B=@n31), B=Multiply(A=@n35, B=@n36)), B=Add(A=Add(A=@n37, B=@n39), B=Multiply(A=@n40, B=@n41)))
  n43 = LinearInterpolate(A=Multiply(A=@n27.RGB, B=@n27.RGB), B=@n29, Alpha=Saturate(None=@n42))
  n44 = QualitySwitch(Default=@n43, Low=@n29, High=@n43, Medium=@n29)
  n45 = LinearInterpolate(
   A=@n26.BaseColor
   B=FunctionOutput(None=ShadingPathSwitch(Default=@n44, Deferred=@n44, Forward=@n29, Mobile=@n29)) {outputName=Color, sortPriority=2}
   Alpha=@n12
  )
  n46 = Saturate(None=Multiply(A=@n42, B=ScalarParameter() {defaultValue=2, group=Foam and Pondscum, parameterName=Foam Sharpen}))
  n47 = Divide(
   A=Subtract(A=@n17, B=ScalarParameter() {defaultValue=69.7077, parameterName=Depth Foam Offset_LOW})
   B=ScalarParameter() {defaultValue=300, group=Foam and Pondscum, parameterName=Depth for Foam Edge_LOW}
  )
  n48 = Multiply(A=OneMinus(None=Saturate(None=@n47)), B=Multiply(A=Subtract(A=@n34.A) {constB=0.5}, B=@n36))
  n49 = QualitySwitch(Default=@n46, Low=@n48, High=@n46, Medium=@n48)
  n50 = ShadingPathSwitch(Default=@n49, Deferred=@n49, Forward=@n48, Mobile=@n48)
  n51 = Constant()
  n52 = FunctionOutput(None=Multiply(A=StaticSwitchParameter(True=@n50, False=@n51) {defaultValue=true, parameterName=Use Ocean Foam}, B=@n12)) {outputName=Mask, sortPriority=1}
  n53 = LinearInterpolate(A=@n26.Opacity, B=ScalarParameter() {defaultValue=2, group=Foam and Pondscum, parameterName=Foam Opacity}, Alpha=@n52)
  n54 = ComponentMask(None=@n34.RGB) {g=true, r=true}
  n55 = Multiply(
   A=AppendVector(A=ConstantBiasScale(None=@n54) {bias=-0.5, scale=2}, B=@n34.B)
   B=VectorParameter() {parameterName=FoamNormal, defaultValue=<rgba 1 1 1 0>}.RGB
  )
  n56 = SetMaterialAttributes!(
   MaterialAttributes=@n26.MaterialAttributes
   BaseColor=@n45
   Specular=LinearInterpolate(A=@n26.Specular, B=ScalarParameter() {defaultValue=-0.5, group=Foam and Pondscum, parameterName=Foam Specular}, Alpha=@n52)
   Roughness=LinearInterpolate(A=@n26.Roughness, B=ScalarParameter() {defaultValue=0.5, group=Foam and Pondscum, parameterName=Foam Roughness}, Alpha=@n52)
   Opacity=@n53
   Normal=LinearInterpolate(A=@n26.Normal, B=FunctionOutput(None=@n55) {outputName=Normal, sortPriority=4}, Alpha=@n52)
  ) {attributeSetTypes [
    #2
    #3
    #4
    #5
    #6
   ]}
  n57 = SetMaterialAttributes!(MaterialAttributes=@n26.MaterialAttributes, BaseColor=@n45, Opacity=@n53) {attributeSetTypes [#2, #5]}
  n58 = QualitySwitch(Default=@n56, Low=@n57, High=@n56, Medium=@n57)
  FunctionOutput(None=ShadingPathSwitch(Default=@n58, Deferred=@n58, Forward=@n57, Mobile=@n57)) {outputName=Attributes}
  n59 = MaterialFunctionCall(
   Time=MaterialFunctionCall() {materialFunction=&GetWaterTime}.WaterTime
   Speed=ScalarParameter() {defaultValue=0.2, parameterName=Foam Bob Speed}
  ) {materialFunction=&TimeWithSpeedVariable}
  Reroute(None=Multiply(A=@n10, B=Sine(None=@n59.Time)))
  TextureSample(UVs=@n33, Bias=ScalarParameter() {group=Foam and Pondscum, parameterName=Foam Normal Blur}) {
   mipValueMode=TMVM_MipBias
   samplerSource=SSM_Wrap_WorldGroupSettings
   samplerType=SAMPLERTYPE_Normal
   texture=/Game/Global/Textures/Default/Tileable_Patterns/T_TRV_Moss_01_N
  }
 ]}
