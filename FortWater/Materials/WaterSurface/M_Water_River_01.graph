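# Masked, two-sided Single Layer Water material for Fortnite river surfaces, with static switches for ocean, swamp and cave transitions and a waterfall variant.
# Velocity-masked Gerstner waves, panned normals, height-mapped offset and simulation feed flowmap, foam, pond scum, Styx, vortex and liquid-gold layers.
# River velocity blends scattering and absorption, tinted by biome, noise and macro masks.
# Depth fades edges, dynamic bounds clip the surface, and waterfalls use vertex-color flow and dithered alpha.
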Material /FortWater/Materials/WaterSurface/M_Water_River_01
graph {
 properties {
  allowTranslucentCustomDepthWrites=true
  bTangentSpaceNormal=false
  bUseMaterialAttributes=true
  bUsedWithInstancedStaticMeshes=true
  bUsedWithSplineMeshes=true
  bUsedWithWater=true
  blendMode=BLEND_Masked
  pixelDepthOffsetMode=PDOM_Legacy
  refractionMethod=RM_PixelNormalOffset
  shadingModel=MSM_SingleLayerWater
  twoSided=true
  usedShadingModels=SingleLayerWater
 }
 nodes [
  n1 = VectorParameter() {group=Water Emissive, parameterName=EmissiveColor, channelNames {a=DepthGradient, b="", g="", r=""}, defaultValue=<rgba 0 0 0 1>}
  n2 = VectorParameter() {group=Water Emissive, parameterName=RiverEmissiveColor, channelNames {a=(unused), b="", g="", r=""}, defaultValue=<rgba 0 0 0 1>}
  n3 = LinearInterpolate(A=@n1, B=@n2, Alpha=NamedRerouteUsage())
  n4 = Saturate(None=Divide(A=NamedRerouteUsage(), B=@n1.DepthGradient))
  n5 = VectorParameter() {group=Water Emissive, parameterName=EmissiveTODFactors, channelNames {a=Night, b=Evening, g=Day, r=Morning}, defaultValue=<rgba 1 1 1 1>}
  n6 = Distance(A=CameraPositionWS(), B=WorldPosition().XYZ)
  n7 = VectorParameter() {
   group=Water Emissive
   parameterName=EmissiveDistanceFadeParameters
   channelNames {a=InvertFade, b=MinFade, g=DistanceOffset, r=FadeDistance}
   defaultValue=<rgba 10000 0 0 1>
  }
  n8 = Saturate(None=Divide(A=Add(A=@n6, B=@n7.DistanceOffset), B=@n7.FadeDistance))
  n9 = OneMinus(None=@n8)
  n10 = Multiply(
   A=Multiply(
    A=Multiply(A=@n3, B=@n4)
    B=MaterialFunctionCall(Value_Morning=@n5.Morning, Value_Day=@n5.Morning, Value_Evening=@n5.Morning, Value_Night=@n5.Morning) {materialFunction=&Scalar_Time_of_Day}.Result
   )
   B=Max(A=LinearInterpolate(A=@n9, B=@n8, Alpha=@n7.InvertFade), B=@n7.MinFade)
  )
  n11 = Divide(B=EyeAdaptation().EyeAdaptation)
  n12 = ScalarParameter() {defaultValue=0.225, group=Water Shading Model, parameterName=Specular}
  n13 = SceneDepthWithoutWater!(UVs)
  n14 = PixelDepth()
  n15 = QualitySwitch(
   Default=MaterialFunctionCall!(InSpecular=@n12, WaterDepth=Subtract(A=@n13.R, B=@n14.R)) {materialFunction=/FortWater/Materials/Functions/MF_WaterBodySpecular}.Result
   Low=@n12
   Medium=@n12
  )
  n16 = NamedRerouteUsage()
  n17 = LinearInterpolate(
   A=@n15
   B=ScalarParameter() {defaultValue=0.05, group=MacroColorVariation, parameterName=Specular_MacroMask}
   Alpha=Power(
    Base=StaticComponentMaskParameter(None=@n16) {defaultR=true, group=MacroColorVariation, parameterName=MacroMaskChannelForSpecular}
    Exp=ScalarParameter() {defaultValue=1, group=MacroColorVariation, parameterName=MacroMaskSpecular_Power}
   )
  )
  n18 = MaterialFunctionCall!(
   Z Offset Only=StaticBoolParameter() {defaultValue=true, parameterName="UseSimpleGestnerWaves(ZOffsetOnly)"}
   Position
   WaterBodyIndex=StaticSwitchParameter(
    True=ScalarParameter() {parameterName=WaterfallOverrideWaterBodyIndex}
    False=Custom() {code="GetWaterWaveParamIndex(Parameters)", description=WaterBodyIndex From Tile Instance, outputType=CMOT_Float1, inputs []}
   ) {desc=Let custom water body can get correct wave information, parameterName=IsWaterfallMaterial}
   Normals use VS Interpolation
   All Normals VS
  ) {materialFunction=/FortWater/Materials/Functions/Fort_ComputeGerstnerWaves}
  n19 = NamedRerouteUsage()
  n20 = NamedRerouteUsage()
  n21 = VectorParameter() {parameterName=Cave Loc, defaultValue=<rgba -24760 122936 1 0>}
  n22 = SphereMask(
   A=ComponentMask(None=WorldPosition().XYZ) {g=true, r=true}
   B=ComponentMask(None=@n21.RGB) {g=true, r=true}
   Radius=@n21.B
   Hardness=Constant() {r=0.5}
  ) {desc=Swamp Pondscum Mask, hardnessPercent=50}
  n23 = ComponentMask(None=StaticSwitchParameter(True=@n22, False=NamedRerouteUsage()) {parameterName=Cave River Transition}) {b=true}
  n24 = ScalarParameter() {defaultValue=8, parameterName=Velocity for Full Waves}
  n25 = Saturate(None=Divide(A=Subtract(A=@n23, B=@n24), B=ScalarParameter() {defaultValue=64, parameterName=Velocity for No Waves}))
  n26 = Multiply(A=NamedRerouteDeclaration(None=OneMinus(None=@n25)) {name=Inverted River Mask, nodeColor=<rgba 1 0.4688 0 1>}, B=NamedRerouteUsage())
  n27 = StaticSwitchParameter(True=Multiply(A=Saturate(None=@n20), B=@n26), False=@n26) {parameterName=Swamp River Transition}
  n28 = MaterialFunctionCall!(
   Normal=@n18.Normal
   WPO=@n18.Normal
   Mask=StaticSwitchParameter(True=Multiply(A=@n19, B=@n27), False=@n27) {parameterName=Cave River Transition}
  ) {materialFunction=/FortWater/Materials/Functions/MaskOutWaves}
  MaterialExpressionMaterialFunctionCall_6 = MaterialFunctionCall!(
   Normal=StaticSwitchParameter(True=@n28.Normal, False=Constant3Vector() {constant=<rgba 0 0 1 0>}) {parameterName=River Transition}
   WPO=@n28.WPO
   Far Mask Override=MaterialFunctionCall(
    Fade Length=ScalarParameter() {defaultValue=15000, group=Distance Blending, parameterName=Far Fade Length}
    Fade Offset=ScalarParameter() {defaultValue=7500, group=Distance Blending, parameterName=Far Fade Start}
   ) {materialFunction=&CameraDepthFade}.Result
   Additional Mask=StaticSwitchParameter(True=NamedRerouteUsage(), False=Constant()) {parameterName=River Transition}
   LowDetailBiome_Mask=NamedRerouteUsage()
  ) {materialFunction=/FortWater/Materials/Functions/WaterTwoTexturePanner}
  MaterialExpressionSaturate_11 = Saturate(None=NamedRerouteUsage())
  n29 = Custom!(NormalPassThru=@MaterialExpressionMaterialFunctionCall_6.Normal, HoleMask=@MaterialExpressionSaturate_11) {
   code="##if MATERIALBLENDING_MASKED\r\nif (HoleMask == 0)\r\n  clip(-1);\r\n##endif\r\nreturn NormalPassThru;"
   description=EarlyOut
   inputs [
    {inputName=NormalPassThru, input {expression=&self:MaterialExpressionMaterialFunctionCall_6}}
    {inputName=HoleMask, input {expression=&self:MaterialExpressionSaturate_11}}
   ]
  }
  n30 = Constant()
  n31 = StaticSwitchParameter(True=StaticSwitchParameter(True=@n25, False=Constant() {r=1}) {parameterName=River Transition}, False=Constant() {r=1}) {
   desc="Enables a velocity mask which prevents height mapping on traisitions \r\nbetween rivers and lakes which have height mapping disabled.\r\nThis was disabled by default when enabling height mapping on all lakes"
   parameterName=UseVelocityMask
  }
  n32 = Subtract(A=NamedRerouteUsage(), B=ComponentMask(None=WorldPosition() {worldPositionShaderOffset=WPT_ExcludeAllShaderOffsets}.XYZ) {b=true})
  n33 = ScalarParameter() {defaultValue=4, parameterName=Max Z Displacement}
  n34 = Multiply(A=@n31, B=StaticSwitchParameter(True=@n32, False=Min(A=@n32, B=@n33) {constB=0}) {parameterName=River Transition})
  n35 = Constant3Vector() {constant=<rgba 0 0 1 0>}
  n36 = StaticSwitchParameter(
   True=@n30
   False=StaticSwitchParameter(True=Multiply(A=@n34, B=@n35), False=@n30) {defaultValue=true, parameterName=UseHeightMappingVertexShader}
  ) {desc=Let custom water body can get correct wave information, parameterName=IsWaterfallMaterial}
  n37 = ScalarParameter() {defaultValue=35000, parameterName=ZOffsetOffset}
  n38 = Subtract(A=CameraPositionWS(), B=Add(A=@n36, B=WorldPosition() {worldPositionShaderOffset=WPT_ExcludeAllShaderOffsets}.XYZ))
  n39 = ScalarParameter() {defaultValue=256, parameterName=DistanceCameraOffset}
  n40 = LinearInterpolate(
   A=@n39
   B=ScalarParameter() {parameterName=DistanceCameraOffset_Ocean}
   Alpha=Power(Base=NamedRerouteUsage(), Exp=ScalarParameter() {defaultValue=3, parameterName=DistanceCameraOffset_Ocean_Contrast})
  )
  n41 = StaticSwitchParameter(True=@n40, False=@n39) {parameterName=Ocean River Transition}
  n42 = Multiply(
   A=MaterialFunctionCall(Fade Length=@n37, Fade Offset=@n37) {materialFunction=&CameraDepthFade}.Result
   B=Multiply(A=Normalize(VectorInput=@n38), B=StaticSwitchParameter(True=@n40, False=@n41) {parameterName=River Transition})
  )
  n43 = MaterialFunctionCall!(
   Flowmap Velocity=NamedRerouteUsage()
   Previous Foam
   Normal=StaticSwitchParameter(True=@MaterialExpressionMaterialFunctionCall_6.Normal, False=@n29) {desc=Let custom water body can get correct wave information, parameterName=IsWaterfallMaterial}
   WPO=Add(A=@MaterialExpressionMaterialFunctionCall_6.WPO, B=Add(A=@n36, B=@n42) {constB=0})
  ) {materialFunction=/FortWater/Materials/Functions/WaterSimulationTexture}
  n44 = SetMaterialAttributes!(
   MaterialAttributes
   Base Color=VectorParameter() {parameterName=BaseColor, defaultValue=<rgba 0 0.1 0.1 0.1>}.RGB
   Emissive Color=StaticSwitchParameter(True=Multiply(A=@n10, B=@n11), False=Constant()) {parameterName=UseEmissive}
   Specular=StaticSwitchParameter(True=StaticSwitchParameter(True=@n17, False=@n15) {group=MacroColorVariation, parameterName=UseMacroColorMaskToBlendSpecular}, False=@n15) {parameterName=UseMacroWaterColorVariation}
   Roughness=ScalarParameter() {defaultValue=0.05, group=Water Shading Model, parameterName=Roughness}
   Opacity=ScalarParameter() {group=Water Shading Model, parameterName=Surface Opacity}
   Normal=@n43.Normal
   World Position Offset=@n43.Normal
  ) {attributeSetTypes [
    #1
    #2
    #3
    #4
    #5
    #6
    #7
   ]}
  n45 = NamedRerouteUsage()
  n46 = MaterialFunctionCall!(
   Attributes=@n44
   Transition=NamedRerouteUsage()
   Sim Foam Erasure=@n43.Sim Foam Erasure
   Velocity=NamedRerouteUsage()
   Far Mask Override
   Depth=@n45
  ) {materialFunction=/FortWater/Materials/Functions/FortWaterRiverFlowmaps}
  n47 = StaticSwitchParameter(True=@n46.Attributes, False=@n46.Attributes) {desc=Let custom water body can get correct wave information, parameterName=IsWaterfallMaterial}
  n48 = VectorParameter() {group=Biomes, parameterName=Foam Mask Location Radius, defaultValue=<rgba 81160 -98924 10000 1>}
  n49 = SphereMask(
   A=ComponentMask(None=WorldPosition().XYZ) {g=true, r=true}
   B=ComponentMask(None=@n48.RGB) {g=true, r=true}
   Radius=@n48.B
   Hardness=ScalarParameter() {defaultValue=0.85, group=Biomes, parameterName=Foam Mask Hardness}
  ) {desc=Custom Foam Mask}
  n50 = Constant()
  n51 = NamedRerouteUsage()
  n52 = StaticSwitchParameter(
   True=NamedRerouteUsage()
   False=StaticSwitchParameter(True=Constant(), False=StaticSwitchParameter(True=@n51, False=@n50) {parameterName=Ocean River Transition}) {parameterName=Swamp River Transition}
  ) {parameterName=Cave River Transition}
  n53 = StaticSwitchParameter(True=Multiply(A=OneMinus(None=@n49), B=Saturate(None=@n52)), False=@n52) {parameterName=UseFoamMask}
  n54 = MaterialFunctionCall!(
   Attributes=@n47
   Sim Erasure=@n43.Sim Foam Erasure
   Mask=@n53
   Far Mask
   Depth=@n45
   WaveHeight=ComponentMask(None=NamedRerouteUsage()) {b=true}
  ) {materialFunction=/FortWater/Materials/Functions/OceanShoreFoam_2}
  n55 = MaterialFunctionCall!(Attributes=@n47, Sim Erasure=@n43.Sim Foam Erasure, Mask=@n53, Far Mask, Depth=@n45) {materialFunction=/FortWater/Materials/Functions/OceanShoreFoam}
  n56 = Multiply(A=StaticSwitchParameter(True=@n51, False=Constant() {r=1}) {parameterName=Ocean River Transition}, B=OneMinus(None=NamedRerouteUsage()))
  n57 = Constant()
  n58 = MaterialFunctionCall!(
   DefaultMaterial=MaterialFunctionCall!(
    Attributes=StaticSwitchParameter(True=@n54.Attributes, False=@n55.Attributes) {defaultValue=true, parameterName=UseNewOceanFoam2}
    Sim Erasure=@n43.Sim Foam Erasure
    Alpha Mask=StaticSwitchParameter(True=@n56, False=@n57) {parameterName=Swamp River Transition}
    Depth=@n45
   ) {materialFunction=/FortWater/Materials/Functions/PondScumLayer}.Attributes
   Material_NoFoam=@n47
   Depth=@n45
   SimWaterNormal
   SimErasure=@n43.Sim Foam Erasure
   TransitionMask=NamedRerouteUsage()
   Scattering=NamedRerouteUsage()
   Absorption=NamedRerouteUsage()
   Is Translucent Mat
   Mobile Reflections
  ) {materialFunction=/FortWater/Materials/Functions/MF_Water_Styx}
  n59 = NamedRerouteUsage()
  n60 = MaterialFunctionCall!(
   Attributes=MaterialFunctionCall!(
    In=MaterialFunctionCall!(In=@n58.Result, MobileNormalIn, MobileFoamMaskIn) {materialFunction=/FortWater/Materials/Functions/MF_WaterVortex}.Result
    Velocity=NamedRerouteUsage()
    Depth=NamedRerouteUsage()
    Transition=NamedRerouteUsage()
    IsRiverMaterial=StaticBool() {value=true}
    IsTranslucentMat
   ) {materialFunction=/FortWater/Materials/WaterSurface/Hermes/Gold/MF_LiquidGold}.Result
   Depth=@n59
  ) {materialFunction=/FortWater/Materials/Functions/TwoSidedWaterCorrections}
  n61 = ScalarParameter() {defaultValue=-180, parameterName=Water opacity Offset}
  n62 = PixelDepth()
  n63 = ScalarParameter() {defaultValue=25000, parameterName=Water Opacity Offset Fade Distance}
  n64 = LinearInterpolate(A=@n61, B=ScalarParameter() {parameterName=Water opacity Offset Far}, Alpha=Saturate(None=Divide(A=@n62.R, B=@n63) {constB=15000})) {constB=0}
  n65 = NamedRerouteDeclaration(None=Subtract(A=@n59, B=StaticSwitchParameter(True=@n64, False=@n61) {defaultValue=true, parameterName=Use Distance Opacity Bias})) {name=Water Depth Distance Biased, nodeColor=<rgba 0.2188 0 1 1>}
  n66 = Divide(A=@n65, B=ScalarParameter() {defaultValue=48, parameterName=Water Opacity Gradient})
  n67 = StaticSwitchParameter(True=Constant() {r=1}, False=StaticSwitchParameter(True=Constant() {r=1}, False=Saturate(None=@n66)) {parameterName=DisableOpacityMask}) {desc=Let custom water body can get correct wave information, parameterName=IsWaterfallMaterial}
  n68 = Subtract(A=SceneDepthWithoutWater().R, B=PixelDepth().R)
  n69 = Divide(A=@n68, B=ScalarParameter() {defaultValue=32, parameterName=Depth Edge Gradient})
  n70 = MaterialFunctionCall!(
   Attributes=MaterialFunctionCall!(
    In=MaterialFunctionCall!(Attributes=@n60.Result) {materialFunction=/FortWater/Materials/Functions/WaterQualityOptimizations}.Result
    TransitionMask=@n58.TransitionMask
    Is Translucent Mat
   ) {materialFunction=/FortWater/Materials/Functions/MF_Water_Styx_SpecularOverride}.Result
   Opacity Mask=Multiply(A=@n67, B=Saturate(None=@n69))
  ) {materialFunction=/FortWater/Materials/Functions/WaterOpacityMultiply}
  n71 = MaterialFunctionCall!(Attributes=@n70.Result) {materialFunction=/FortWater/Materials/Functions/WaterPhysicsForce}
  Composite!(InMA=@n71.Result) {desc=to match normal intensity of Epic, subgraphName=NormalizeNormal}
  Constant3Vector() {constant=<rgba 0.1268 0.1336 0.15 0.5>}
  Constant3Vector() {constant=<rgba 1 13.35 20 100>}
  Constant()
  MaterialExpressionReroute_18 = Reroute(None=@n71.Result)
  n72 = SetMaterialAttributes!(
   MaterialAttributes=@MaterialExpressionReroute_18.InMA
   Normal=Normalize(VectorInput=GetMaterialAttributes(None=@MaterialExpressionReroute_18.InMA) {attributeGetTypes [#6]}.Normal)
  ) {attributeSetTypes [#6]}
  MaterialExpressionReroute_20 = Reroute(None=@n72)
  n73 = QualitySwitch(
   Default=@n70.Result
   High=@MaterialExpressionReroute_20.OutMA
   Medium=@MaterialExpressionReroute_20.OutMA
   Epic=@MaterialExpressionReroute_20.OutMA
  )
  n74 = MaterialFunctionCall!(Attributes=@n73) {materialFunction=/FortWater/Materials/Functions/ClipWaterHitProxyFromDepth}
  n75 = StaticSwitchParameter(True=@n73, False=MaterialFunctionCall!(Attributes=@n74.Result) {materialFunction=/Water/Materials/Functions/ClipDynamicBounds}.Attributes) {desc=Let custom water body can get correct wave information, parameterName=IsWaterfallMaterial}
  n76 = VertexInterpolator(VS=VertexColor().A)
  n77 = Constant() {r=1}
  n78 = ScalarParameter() {defaultValue=0.5, parameterName=WaterfallDitherRandom}
  n79 = SetMaterialAttributes!(
   MaterialAttributes=@n75
   Opacity Mask=If(
    A=@n76.PS
    "A > B"=@n77
    "A == B"=@n77
    "A < B"=MaterialFunctionCall(Alpha Threshold=@n76.PS, Random=@n78) {materialFunction=&DitherTemporalAA_Fortnite}.Result
   ) {constB=0.99}
  ) {attributeSetTypes [#8]}
  MaterialFunctionCall!(Material Attributes=MaterialFunctionCall!(InMA=StaticSwitchParameter(True=@n79, False=@n75) {desc=Let custom water body can get correct wave information, parameterName=IsWaterfallMaterial}) {materialFunction=/Game/Tools/NewWaterfallGenerator/Material/WaterfallClipBound}.Result) {materialFunction=&/MF_Water_FarSpecular}
  n80 = ComponentMask(None=CollectionParameter() {collection=/Game/Athena/SafeZone/MPC_SafeZoneClouds, parameterName=SkyfireLoc}) {g=true, r=true}
  n81 = ComponentMask(None=WorldPosition().XYZ) {g=true, r=true}
  n82 = ScalarParameter() {defaultValue=160656.3125, parameterName=shipShadowRadius}
  n83 = ScalarParameter() {defaultValue=0.9, parameterName=shipShadowHardness}
  n84 = CameraPositionWS()
  n85 = Divide(
   A=Subtract(
    A=MaterialFunctionCall(Float3=@n84) {materialFunction=&BreakOutFloat3Components}.B
    B=ScalarParameter() {defaultValue=12000, parameterName=ShipShadowFadeHeightBias}
   )
   B=ScalarParameter() {defaultValue=7000, parameterName=ShipShadowFadeHeightScale}
  )
  NamedRerouteDeclaration(None=Multiply(A=SphereMask(A=@n80, B=@n81, Radius=@n82, Hardness=@n83), B=Saturate(None=@n85))) {name=ShipShadow, nodeColor=<rgba 0.8516 0 1 1>}
  n86 = VectorParameter() {group=Water Shading Model, parameterName=OceanAbsorption, defaultValue=<rgba 1 13.35 20 100>}
  NamedRerouteDeclaration(None=Multiply(A=@n86.RGB, B=@n86.RGB)) {name=Ocean Absorbtion, nodeColor=<rgba 0 1 1 1>}
  n87 = VectorParameter() {group=Water Shading Model, parameterName=Scattering, defaultValue=<rgba 0.1268 0.1336 0.15 0.5>}
  n88 = Multiply(A=@n87.RGB, B=Divide(A=@n87.A) {constB=100})
  n89 = VectorParameter() {group=Water Shading Model, parameterName=Scattering_Noise, defaultValue=<rgba 0.1268 0.1336 0.15 0.5>}
  n90 = WorldPosition()
  n91 = TextureSample(UVs=Divide(
    A=ComponentMask(None=@n90.XYZ) {g=true, r=true}
    B=ScalarParameter() {defaultValue=130000, group=Water Shading Model, parameterName=OceanColorNoise_Size}
   )) {samplerType=SAMPLERTYPE_Masks, texture=/FortWater/Textures/OceanNoise__Mask}
  n92 = LinearInterpolate(A=@n88, B=Multiply(A=@n89.RGB, B=Divide(A=@n89.A) {constB=100}), Alpha=@n91.R)
  NamedRerouteDeclaration(None=StaticSwitchParameter(True=ShadingPathSwitch(Default=@n92, Mobile=@n88) {desc=Migrated from FeatureLevelSwitch}, False=@n88) {parameterName=UseOceanColorNoise}) {name=Scatter, nodeColor=<rgba 0.6172 0 1 1>}
  n93 = VectorParameter() {group=Water Shading Model, parameterName=Absorption, defaultValue=<rgba 1 13.35 20 100>}
  n94 = VectorParameter() {
   group=Water Shading Model
   parameterName=AbsorptionDistanceSettings
   channelNames {a="", b=DistanceRadius, g=DistanceOffset, r=UseDistanceAbsorption}
   defaultValue=<rgba 0 0 30000 1>
  }
  n95 = SphereMask(
   A=Add(A=CameraPositionWS(), B=@n94.DistanceOffset)
   B=WorldPosition() {worldPositionShaderOffset=WPT_ExcludeAllShaderOffsets}.XYZ
   Radius=@n94.DistanceRadius
   Hardness=Constant()
  )
  n96 = Multiply(A=@n95, B=@n95)
  n97 = OneMinus(None=@n94.UseDistanceAbsorption)
  n98 = LinearInterpolate(
   A=VectorParameter() {group=Water Shading Model, parameterName=AbsorptionDistance, defaultValue=<rgba 1 1 1 1>}.RGB
   B=Multiply(A=@n93.RGB, B=@n93.RGB)
   Alpha=Saturate(None=Add(A=@n96, B=@n97))
  )
  n99 = VectorParameter() {group=Water Shading Model, parameterName=Absorption_Noise, defaultValue=<rgba 1 13.35 20 100>}
  n100 = LinearInterpolate(A=@n98, B=Multiply(A=@n99.RGB, B=@n99.RGB), Alpha=@n91.R)
  NamedRerouteDeclaration(None=StaticSwitchParameter(True=ShadingPathSwitch(Default=@n100, Mobile=@n98) {desc=Migrated from FeatureLevelSwitch}, False=@n98) {parameterName=UseOceanColorNoise}) {name=Absorption, nodeColor=<rgba 0.3125 1 0 1>}
  NamedRerouteDeclaration(None=@n18.WPO) {name=GerstnerWave_WPO, nodeColor=<rgba 0.1953 1 0 1>}
  n101 = MaterialFunctionCall!(Texture, Position) {materialFunction=/FortWater/Materials/Functions/WaterBodyData}
  NamedRerouteDeclaration(None=@n101.VelocityAndLength) {name=WaterBody_VelocityandLength, nodeColor=<rgba 0.5937 1 0 1>}
  n102 = ScalarParameter() {parameterName=Max Z Location}
  n103 = WorldPosition()
  n104 = VectorParameter() {parameterName=Z Clamp Center and Length, defaultValue=<rgba 0 0 0 100>}
  n105 = Divide(A=StaticComponentMaskParameter(None=Subtract(A=@n103.XYZ, B=@n104.RGB)) {defaultG=true, parameterName=Mask Gradient Direction}, B=@n104.A)
  n106 = StaticSwitchParameter(
   True=LinearInterpolate(A=@n101.Water Z Height, B=Min(A=@n101.Water Z Height, B=@n102), Alpha=Saturate(None=@n105))
   False=@n101.Water Z Height
  ) {parameterName=Use Custom Max Z}
  NamedRerouteDeclaration(None=@n106) {name=Water Z Height WithCustom Max Z, nodeColor=<rgba 1 0.7266 0 1>}
  n107 = MaterialFunctionCall!(WaterDepth=NamedRerouteDeclaration(None=@n101.Water Depth) {name=Water Depth, nodeColor=<rgba 0 1 0.1094 1>}, Water Body Index) {materialFunction=/Water/Materials/Functions/ComputeWavesAttenuationFactor}
  NamedRerouteDeclaration(None=@n107.WaveAttenuationFactor) {name=Water Depth With Attenuation, nodeColor=<rgba 1 0 0.2344 1>}
  NamedRerouteDeclaration(None=@n91.R) {name=OceanNoise, nodeColor=<rgba 1 0 0.7031 1>}
  n108 = LinearInterpolate(A=NamedRerouteUsage(), B=NamedRerouteUsage(), Alpha=NamedRerouteUsage())
  n109 = Add(A=NamedRerouteUsage(), B=@n108)
  n110 = QualitySwitch(Default=@n109, Low=@n108, High=@n109, Medium=@n108)
  n111 = VectorParameter() {group=Water Shading Model, parameterName=RiverScattering, defaultValue=<rgba 0.1268 0.1336 0.15 0.5>}
  n112 = VectorParameter() {group=Water Shading Model, parameterName=BiomeRiverScattering, defaultValue=<rgba 0.1268 0.1336 0.15 0.5>}
  n113 = NamedRerouteUsage()
  n114 = LinearInterpolate(A=@n111.A, B=@n112.A, Alpha=@n113)
  n115 = Multiply(
   A=StaticSwitchParameter(True=LinearInterpolate(A=@n111.RGB, B=@n112.RGB, Alpha=@n113), False=@n111.RGB) {group=BiomeMask, parameterName=UseBiomeMask}
   B=Divide(A=StaticSwitchParameter(True=@n114, False=@n111.A) {group=BiomeMask, parameterName=UseBiomeMask}) {constB=100}
  )
  n116 = Add(A=NamedRerouteUsage(), B=@n115)
  n117 = QualitySwitch(Default=@n116, Low=@n115, High=@n116, Medium=@n115)
  n118 = VectorParameter() {group=Water Shading Model, parameterName=RiverAbsorption, defaultValue=<rgba 1 13.35 20 100>}
  n119 = VectorParameter() {group=Water Shading Model, parameterName=BiomeRiverAbsorption, defaultValue=<rgba 1 13.35 20 100>}
  n120 = NamedRerouteUsage()
  n121 = LinearInterpolate(A=@n118.RGB, B=@n119.RGB, Alpha=@n120)
  n122 = LinearInterpolate(A=@n118.A, B=@n119.A, Alpha=@n120)
  n123 = MaterialFunctionCall() {materialFunction=&WaterBodyData}
  n124 = Saturate(None=Divide(A=ComponentMask(None=@n123.VelocityAndLength) {b=true}, B=ScalarParameter() {defaultValue=48, parameterName=Velocity for Scattering}))
  n125 = MaterialFunctionCall!(
   Scattering A=ShadingPathSwitch(Default=@n110, Deferred=@n110, Forward=@n108, Mobile=@n108)
   Scattering B=ShadingPathSwitch(Default=@n117, Deferred=@n117, Forward=@n115, Mobile=@n115)
   Absorption A=LinearInterpolate(A=NamedRerouteUsage(), B=NamedRerouteUsage(), Alpha=NamedRerouteUsage())
   Absorption B=Multiply(
    A=StaticSwitchParameter(True=@n121, False=@n118.RGB) {group=BiomeMask, parameterName=UseBiomeMask}
    B=StaticSwitchParameter(True=@n122, False=@n118.A) {group=BiomeMask, parameterName=UseBiomeMask}
   )
   Alpha=@n124
  ) {materialFunction=/FortWater/Materials/Functions/BlendWaterShadingValues}
  n126 = NamedRerouteUsage()
  n127 = MaterialFunctionCall!(
   InScattering=@n125.Scattering
   InAbsorption=@n125.Scattering
   InTransitionMask=NamedRerouteUsage()
   InOceanExtensionMask
   InNoise=StaticSwitchParameter(True=LinearInterpolate(A=@n126, Alpha=@n124), False=Constant() {r=1}) {parameterName=UseOceanColorNoise}
  ) {materialFunction=/FortWater/Materials/Functions/MF_WaterMacroColorVariation}
  NamedRerouteDeclaration(None=@n127.Mask) {name=MacroMaskPacked, nodeColor=<rgba 0.6172 1 0 1>}
  n128 = ScalarParameter() {defaultValue=200, parameterName=WaterfallDepth}
  n129 = NamedRerouteUsage()
  n130 = ComponentMask(None=@n129) {b=true}
  n131 = ScalarParameter() {defaultValue=300, parameterName=WaterfallVelocityBlendingWidth}
  n132 = Power(Base=Saturate(None=Divide(A=@n130, B=@n131)), Exp=ScalarParameter() {defaultValue=2, parameterName=WaterfallVelocityBlendingPower})
  n133 = NamedRerouteUsage()
  NamedRerouteDeclaration(None=StaticSwitchParameter(True=LinearInterpolate(A=@n128, B=@n133, Alpha=@n132), False=@n133) {desc=Let custom water body can get correct wave information, parameterName=IsWaterfallMaterial}) {name=WaterfallSupport_WaterDepth, nodeColor=<rgba 0.7109 1 0 1>}
  n134 = CollectionParameter() {collection=/Water/Materials/MPC/MPC_Water, parameterName=SwampColor}
  n135 = Ceil(None=ComponentMask(None=@n134) {a=true})
  n136 = SphereMask(
   A=ComponentMask(None=WorldPosition().XYZ) {g=true, r=true}
   B=ComponentMask(None=@n134) {g=true, r=true}
   Radius=Multiply(
    A=ScalarParameter() {defaultValue=62000, group=Biomes, parameterName=Swamp Pondscum Radius}
    B=Saturate(None=@n135) {desc="Cancel swamp if radius <= 0"}
   )
   Hardness=CollectionParameter() {collection=&MPC_Water, parameterName=SwampMaskHardness}
  ) {desc=Swamp Pondscum Mask}
  NamedRerouteDeclaration(None=StaticSwitchParameter(True=@n136, False=Constant()) {parameterName=Swamp River Transition}) {name=Swamp Mask, nodeColor=<rgba 1 0 0.4219 1>}
  n137 = VertexColor()
  n138 = AppendVector(A=Cosine(None=@n137.R), B=Sine(None=@n137.R))
  n139 = Multiply(A=@n137.G, B=ScalarParameter() {defaultValue=1024, parameterName=WaterfallMaxFlowVelocity})
  n140 = ScalarParameter() {defaultValue=1024, parameterName=WaterfallMaxWhiteWater}
  n141 = AppendVector(A=Add(A=ComponentMask(None=@n129) {g=true, r=true}, B=Multiply(A=@n138, B=@n139)), B=Max(A=@n130, B=Multiply(A=@n137.B, B=@n140)))
  NamedRerouteDeclaration(None=StaticSwitchParameter(True=LinearInterpolate(A=@n141, B=@n129, Alpha=@n132), False=@n129) {desc=Let custom water body can get correct wave information, parameterName=IsWaterfallMaterial}) {name=WaterfallSupport_VelocityAndLength, nodeColor=<rgba 1 0.4688 0 1>}
  NamedRerouteDeclaration(None=@n58.Scatter) {name=Scatter_Styx, nodeColor=<rgba 1 0.6328 0 1>}
  NamedRerouteDeclaration(None=@n58.Absorbtion) {name=AbsorbtionStyx, nodeColor=<rgba 1 0.9609 0 1>}
  NamedRerouteDeclaration(None=@n58.ColorScale) {name=ColorScale_Styx, nodeColor=<rgba 0.0547 1 0 1>}
  NamedRerouteDeclaration(None=@n124) {name=TransitionMask, nodeColor=<rgba 0.5469 0 1 1>}
  NamedRerouteDeclaration(None=Min(A=@n65, B=@n68)) {name=CombinedDepthBehindWater, nodeColor=<rgba 1 0 0.75 1>}
  n142 = VectorParameter() {group=BiomeMask, parameterName=BiomeMaskLocation_and_Radius, channelNames {a="", b=Radius, g=Y, r=X}, defaultValue=<rgba 0 0 0 1>}
  NamedRerouteDeclaration(None=SphereMask(
    A=WorldPosition().XY
    B=ComponentMask(None=@n142) {g=true, r=true}
    Radius=@n142.Radius
    Hardness=ScalarParameter() {defaultValue=0.5, group=BiomeMask, parameterName=BiomeMask_Hardness}
   )) {name=BiomeMask, nodeColor=<rgba 0 1 0.2031 1>}
  NamedRerouteDeclaration(None=StaticSwitchParameter(True=OneMinus(None=@n136), False=Constant()) {parameterName=Swamp River Transition}) {name=Swamp Mask Inverted, nodeColor=<rgba 1 0 0.4219 1>}
  NamedRerouteDeclaration(None=OneMinus(None=@n22)) {name=Cave Mask Inverted, nodeColor=<rgba 0.6406 0 1 1>}
  NamedRerouteDeclaration(None=StaticSwitchParameter(True=@n54.Scatter Mask, False=@n55.Scatter Mask) {defaultValue=true, parameterName=UseNewOceanFoam2}) {name=OceanFoam Scatter, nodeColor=<rgba 0 0.9531 1 1>}
  NamedRerouteDeclaration(None=@n46.Scattering) {name=River Scatter, nodeColor=<rgba 0.1719 0 1 1>}
  n143 = VectorParameter() {group=Water Shading Model, parameterName=OceanScattering, defaultValue=<rgba 0.1268 0.1336 0.15 0.5>}
  NamedRerouteDeclaration(None=Multiply(A=@n143.RGB, B=Divide(A=@n143.A) {constB=100})) {name=Ocean Scatter, nodeColor=<rgba 0 1 0.5078 1>}
  PinBase() {reroutePins [InMA {expression=&self:MaterialExpressionReroute_18}]}
  PinBase!(OutMA=@n72) {reroutePins [OutMA {expression=&self:MaterialExpressionReroute_20}]}
  QualitySwitch(Default=StaticBool(), Low=StaticBool() {value=true})
  Reroute(None=@n124)
  Reroute(None=@n127.Absorption)
  Reroute(None=@n106)
  Reroute(None=@n107.WaveAttenuationFactor)
  Reroute(None=@n127.Scattering)
  SetMaterialAttributes!(MaterialAttributes=@n74.Result, World Position Offset=NamedRerouteUsage()) {attributeSetTypes [#7]}
  n144 = GetMaterialAttributes(None=@n46.Attributes) {attributeGetTypes [#7]}
  n145 = Constant()
  MaterialExpressionAppend3Vector_1 = MaterialXAppend3Vector!(A=@n145, B=@n145, C=ScalarParameter() {defaultValue=20, parameterName=WaterfallWPOOffset})
  n146 = ScalarParameter() {defaultValue=15, parameterName=WaterfallWPOOffsetSoftWidth}
  n147 = Length(None=@n144.World Position Offset) {is=class MaterialExpressionLength}
  SetMaterialAttributes(
   MaterialAttributes=@n144.MaterialAttributes
   World Position Offset=LinearInterpolate(
    A=@n144.World Position Offset
    B=Subtract(A=@n144.World Position Offset, B=@MaterialExpressionAppend3Vector_1)
    Alpha=SmoothStep(Max=@n146, Value=@n147) {constMax=0.9}
   )
  ) {attributeSetTypes [#7]}
  n148 = ComponentMask(None=CameraVectorWS()) {b=true}
  n149 = LinearInterpolate(A=Saturate(None=Ceil(None=@n148)), B=Multiply(A=@n60.NearMask, B=@n67) {desc=Near Mask on Coefficients}, Alpha=Saturate(None=TwoSidedSign()))
  n150 = NamedRerouteUsage()
  n151 = NamedRerouteDeclaration(None=@n127.Scattering) {name=ScatteringFinal, nodeColor=<rgba 1 0 0.2109 1>}
  n152 = StaticSwitchParameter(True=NamedRerouteUsage(), False=NamedRerouteDeclaration(None=@n127.Absorption) {name=AbsoptionFinal, nodeColor=<rgba 1 0 0.4219 1>}) {parameterName=UseStyxShading}
  SingleLayerWaterMaterialOutput!(
   ScatteringCoefficients=Multiply(A=@n149, B=StaticSwitchParameter(True=@n150, False=@n151) {parameterName=UseStyxShading})
   AbsorptionCoefficients=Multiply(A=@n149, B=Divide(B=@n152))
   PhaseG=CollectionParameter() {collection=&MPC_Water, parameterName=Ansotropy}
   ColorScaleBehindWater
  )
  StaticSwitchParameter(True=@n28.WPO, False=Constant()) {parameterName=River Transition}
 ]
}
